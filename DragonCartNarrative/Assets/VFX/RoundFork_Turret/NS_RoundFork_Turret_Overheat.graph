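# Niagara overheat effect for the RoundFork turret, built from mirrored emitter pairs attached to two sockets of the parent static mesh.
# Start_HotSmoke gives an opening mesh-and-sprite smoke burst, Glow_Decal adds a long-lived orange sprite and decal glow, and delayed Looping_Smoke sustains the plume.
# Spawn groups split each emitter into differently tuned layers via masked inputs and visibility tags.
# Loop-age curves shape smoke rate and motion, while time-of-day scalars adjust opacity.
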NiagaraSystem /DragonCartNarrative/Assets/VFX/RoundFork_Turret/NS_RoundFork_Turret_Overheat
dependencies {
 usedDataInterfaces [
  &NiagaraDataInterfaceCurve
  &NiagaraDataInterfaceStaticMesh
  &NiagaraDataInterfaceArrayFloat3
  &NiagaraDataInterfaceSpriteRendererInfo
 ]
 usedDynamicInputs []
 usedModules [
  &self:SystemSpawnScript.SystemScriptSource.SystemScriptGraph.NiagaraNodeAssignment_0.SetVariables_#1
  &SystemState
 ]
}
emitters {
 Glow_Decal {
  emitterUpdateScript [
   &EmitterState(Life Cycle Mode=::NewEnumerator1, Loop Behavior=::NewEnumerator1, Loop Duration=30, UseLoopDelay=false)
   &SpawnBurst_Instantaneous(Use Spawn Probability=false, UseLoopCountLimit=false)
   &SpawnBurst_Instantaneous(Spawn Group=1, Use Spawn Probability=false, UseLoopCountLimit=false) {module=SpawnBurst_Instantaneous001}
  ]
  particleSpawnScript [
   /SetVariables(Particles.VisibilityTag=&SpawnGroup()) {isSetParameters=true, module=SetVariables_#2}
   &InitializeParticle(
    Color=&MaskLinearColorBySpawnGroup(Masked Linear Color=<rgba 0.599 0.1186 0 1>, Passthrough Linear Color=<rgba 50 5.6414 0 1>)
    Lifetime=&MaskFloatBySpawnGroup(Masked Float=27, Passthrough Float=30)
    Position Offset=&MaskVectorBySpawnGroup(Masked Vector=<0 0 0>, Passthrough Vector=<0 0 5>)
    Position Offset Coordinate Space=::Local
    Sprite Size Mode=::NewEnumerator3
    Uniform Sprite Size=1100
    UsePositionOffset=true
    Write Material Random=true
   )
   &CameraOffset(Camera Offset Amount=6)
   &SpriteFacingAndAlignment(WriteAlignment=true, WriteFacing=true)
   &DynamicMaterialParameters(
    Index 0 Param 1=2
    Index 0 Param 2=50
    Index 0 Param 3=1
    Index 0 Param 4=1
    Module.Index0Float0Write=true
    Module.Index0Float1Write=true
    Module.Index0Float2Write=true
    Module.Index0Float3Write=true
   )
   /Niagara/Modules/Spawn/Location/StaticMeshLocation(
    MaskSpawnGroup=false
    Mesh Sampling Type:ENiagara_StaticSamplingMode=::NewEnumerator1
    SamplePosOffset=true
    Sampled Position Offset:Vector3f=<0 0 -30>
    Socket Sampling Mode:ENiagara_StaticSocketMode=::NewEnumerator2
    Static Mesh=DataInterface() {propertyValues {
      DefaultMesh=null
      InstanceIndex=-1
      LODIndex=0
      PreviewMesh=null
      SoftSourceActor=null
      SourceMode=AttachParent
      bAllowSamplingFromStreamingLODs=false
      bCaptureTransformsPerFrame=true
      bUsePhysicsBodyVelocity=false
      FilteredSockets []
      LODIndexUserParameter {parameter {name=None, typeDefHandle {registeredTypeIndex=108}, varData []}}
      MeshParameterBinding {parameter {name=None, typeDefHandle {registeredTypeIndex=1}, varData []}}
      SectionFilter {allowedMaterialSlots []}
     }}
   )
   &CameraOffset(Camera Offset Amount=30) {module=CameraOffset001}
   /SetVariables(Particles.DecalSize:Vector3f=<400 200 3400>) {isSetParameters=true, module=SetVariables_#3}
   /Niagara/Modules/Spawn/Decal/InitialDecalOrientation(Decal Orientation Mode:ENiagara_MeshOrientationOptions=::NewEnumerator7, Rotation:Vector3f=<0 0 0.25>, Use Rotation Vector=true)
  ]
  particleUpdateScript [
   &ParticleState()
   &ScaleColor(
    Scale Alpha=&MaskFloatBySpawnGroup(
     Masked Float=&FloatFromCurve(FloatCurve=DataInterface() {propertyValues {Curve {keys [
          {interpMode=RCIM_Cubic, tangentMode=RCTM_Break, time=0, value=0}
          {interpMode=RCIM_Cubic, tangentMode=RCTM_Break, time=0.1, value=1}
          {interpMode=RCIM_Cubic, tangentMode=RCTM_Break, time=0.8, value=1}
          {interpMode=RCIM_Cubic, tangentMode=RCTM_Break, time=1, value=0}
         ]}}})
     Passthrough Float=&FloatFromCurve(FloatCurve=DataInterface() {propertyValues {Curve {keys [
          <at 0 0 RCIM_Cubic>
          {interpMode=RCIM_Cubic, tangentMode=RCTM_Break, time=0.3, value=1}
          {arriveTangent=-0.0604, interpMode=RCIM_Cubic, leaveTangent=-0.0604, tangentMode=RCTM_User, time=0.6, value=1}
          {interpMode=RCIM_Cubic, tangentMode=RCTM_Break, time=1, value=0}
         ]}}})
    )
    ScaleA=true
    ScaleRGB=true
   ) {module=ScaleColor001}
   &ScaleColor(
    Scale Alpha=&FNTimeOfDayScalar(Night Scale=&MaskFloatBySpawnGroup(Masked Float=0.7, Passthrough Float=1), Morning Scale=1, Evening Scale=1)
    ScaleA=true
    ScaleRGB=true
   )
  ]
  rendererClasses [&NiagaraSpriteRendererProperties, &NiagaraDecalRendererProperties]
  renderers [
   {rendererClass=&NiagaraSpriteRendererProperties}
   {rendererClass=&NiagaraDecalRendererProperties, rendererIndex=1}
  ]
 }
 Glow_Decal001 {
  emitterUpdateScript [
   &EmitterState(Life Cycle Mode=::NewEnumerator1, Loop Behavior=::NewEnumerator1, Loop Duration=30, UseLoopDelay=false)
   &SpawnBurst_Instantaneous(Use Spawn Probability=false, UseLoopCountLimit=false)
   &SpawnBurst_Instantaneous(Spawn Group=1, Use Spawn Probability=false, UseLoopCountLimit=false) {module=SpawnBurst_Instantaneous001}
  ]
  particleSpawnScript [
   /SetVariables(Particles.VisibilityTag=&SpawnGroup()) {isSetParameters=true, module=SetVariables_#2}
   &InitializeParticle(
    Color=&MaskLinearColorBySpawnGroup(Masked Linear Color=<rgba 0.599 0.1186 0 1>, Passthrough Linear Color=<rgba 50 5.6414 0 1>)
    Lifetime=&MaskFloatBySpawnGroup(Masked Float=27, Passthrough Float=30)
    Position Offset=&MaskVectorBySpawnGroup(Masked Vector=<0 0 0>, Passthrough Vector=<0 0 5>)
    Position Offset Coordinate Space=::Local
    Sprite Size Mode=::NewEnumerator3
    Uniform Sprite Size=1100
    UsePositionOffset=true
    Write Material Random=true
   )
   &CameraOffset(Camera Offset Amount=6)
   &SpriteFacingAndAlignment(WriteAlignment=true, WriteFacing=true)
   &DynamicMaterialParameters(
    Index 0 Param 1=2
    Index 0 Param 2=50
    Index 0 Param 3=1
    Index 0 Param 4=1
    Module.Index0Float0Write=true
    Module.Index0Float1Write=true
    Module.Index0Float2Write=true
    Module.Index0Float3Write=true
   )
   &StaticMeshLocation(
    MaskSpawnGroup=false
    Mesh Sampling Type:ENiagara_StaticSamplingMode=::NewEnumerator1
    SamplePosOffset=true
    Sampled Position Offset:Vector3f=<0 0 -30>
    Socket Index=1
    Socket Sampling Mode:ENiagara_StaticSocketMode=::NewEnumerator2
    Static Mesh=DataInterface() {propertyValues {
      DefaultMesh=null
      InstanceIndex=-1
      LODIndex=0
      PreviewMesh=null
      SoftSourceActor=null
      SourceMode=AttachParent
      bAllowSamplingFromStreamingLODs=false
      bCaptureTransformsPerFrame=true
      bUsePhysicsBodyVelocity=false
      FilteredSockets []
      LODIndexUserParameter {parameter {name=None, typeDefHandle {registeredTypeIndex=108}, varData []}}
      MeshParameterBinding {parameter {name=None, typeDefHandle {registeredTypeIndex=1}, varData []}}
      SectionFilter {allowedMaterialSlots []}
     }}
   )
   &CameraOffset(Camera Offset Amount=30) {module=CameraOffset001}
   /SetVariables(Particles.DecalSize:Vector3f=<400 200 3400>) {isSetParameters=true, module=SetVariables_#3}
   &InitialDecalOrientation(Decal Orientation Mode:ENiagara_MeshOrientationOptions=::NewEnumerator7, Rotation:Vector3f=<0 0 0.25>, Use Rotation Vector=true)
  ]
  particleUpdateScript [
   &ParticleState()
   &ScaleColor(
    Scale Alpha=&MaskFloatBySpawnGroup(
     Masked Float=&FloatFromCurve(FloatCurve=DataInterface() {propertyValues {Curve {keys [
          {interpMode=RCIM_Cubic, tangentMode=RCTM_Break, time=0, value=0}
          {interpMode=RCIM_Cubic, tangentMode=RCTM_Break, time=0.1, value=1}
          {interpMode=RCIM_Cubic, tangentMode=RCTM_Break, time=0.8, value=1}
          {interpMode=RCIM_Cubic, tangentMode=RCTM_Break, time=1, value=0}
         ]}}})
     Passthrough Float=&FloatFromCurve(FloatCurve=DataInterface() {propertyValues {Curve {keys [
          <at 0 0 RCIM_Cubic>
          {interpMode=RCIM_Cubic, tangentMode=RCTM_Break, time=0.3, value=1}
          {arriveTangent=-0.0604, interpMode=RCIM_Cubic, leaveTangent=-0.0604, tangentMode=RCTM_User, time=0.6, value=1}
          {interpMode=RCIM_Cubic, tangentMode=RCTM_Break, time=1, value=0}
         ]}}})
    )
    ScaleA=true
    ScaleRGB=true
   ) {module=ScaleColor001}
   &ScaleColor(
    Scale Alpha=&FNTimeOfDayScalar(Night Scale=&MaskFloatBySpawnGroup(Masked Float=0.7, Passthrough Float=1), Morning Scale=1, Evening Scale=1)
    ScaleA=true
    ScaleRGB=true
   )
  ]
  rendererClasses [&NiagaraSpriteRendererProperties, &NiagaraDecalRendererProperties]
  renderers [
   {rendererClass=&NiagaraSpriteRendererProperties}
   {rendererClass=&NiagaraDecalRendererProperties, rendererIndex=1}
  ]
 }
 Looping_Smoke {
  emitterSpawnScript [/SetVariables(Emitter.SmokeDepthFade=10) {isSetParameters=true, module=SetVariables_#4}]
  emitterUpdateScript [
   &EmitterState(Delay First Loop Only=true, Life Cycle Mode=::NewEnumerator1, Loop Delay=0.4, Loop Duration=30, UseLoopDelay=true)
   &SpawnRate(
    SpawnRate=&Multiply_Float(
     A=70
     B=&FloatFromCurve(
      FloatCurve=DataInterface() {propertyValues {Curve {keys [{interpMode=RCIM_Cubic, tangentMode=RCTM_Break, time=0.025, value=2}, {interpMode=RCIM_Cubic, tangentMode=RCTM_Break, time=0.22, value=1}]}}}
      CurveIndex=$Emitter.NormalizedLoopAge
     )
    )
    Use Spawn Probability=false
   ) {module=SpawnRate001}
   &SpawnRate(SpawnGroup=1, SpawnRate=9, Use Spawn Probability=false)
  ]
  particleSpawnScript [
   /SetVariables(Particles.VisibilityTag=&SpawnGroup()) {isSetParameters=true, module=SetVariables_#2}
   &InitializeParticle(
    Color=&MaskLinearColorBySpawnGroup(Masked Linear Color=<rgba 0.4 0.4 0.4 0.15>, Passthrough Linear Color=<rgba 0.4 0.4 0.4 0.08>)
    Lifetime=&MaskFloatBySpawnGroup(
     Masked Float=&RandomRangeFloat(Minimum=2, Maximum=2.2)
     Passthrough Float=&RandomRangeFloat(Minimum=2.4, Maximum=2.8)
    )
    Position Mode=::NewEnumerator0
    Sprite Rotation Mode=::NewEnumerator2
    Sprite Size Mode=::NewEnumerator3
    Uniform Sprite Size=&MaskFloatBySpawnGroup(
     Masked Float=&RandomRangeFloat(Minimum=260, Maximum=340)
     Passthrough Float=&RandomRangeFloat(Minimum=320, Maximum=380)
    )
    Write Material Random=true
   )
   &DynamicMaterialParameters(
    Index 0 Param 1=1
    Module.Index0Float0Write=true
    Module.Index0Float1Write=true
    Module.Index0Float2Write=true
    Module.Index0Float3Write=true
   )
   &AddVelocity(Velocity=&MaskVectorBySpawnGroup(
     Masked Vector=&RandomRangeVector(Minimum=<400 -160 40>, Maximum=<600 160 70>)
     Passthrough Vector=<0 0 0>
    ))
   &AddVelocity(
    Cone Angle=360
    Cone Axis=<0 0 1>
    Module.UseConeFalloff=false
    Module.UseVelDistribution=false
    Override Seed=false
    Velocity Mode=::NewEnumerator2
    Velocity Speed=&MaskFloatBySpawnGroup(Masked Float=0, Passthrough Float=&RandomRangeFloat(Minimum=1600, Maximum=2000))
   ) {module=AddVelocity001}
   &StaticMeshLocation(
    MaskSpawnGroup=false
    Mesh Sampling Type:ENiagara_StaticSamplingMode=::NewEnumerator1
    SamplePosOffset=false
    Socket Sampling Mode:ENiagara_StaticSocketMode=::NewEnumerator2
    Static Mesh=DataInterface() {propertyValues {
      DefaultMesh=null
      InstanceIndex=-1
      LODIndex=0
      PreviewMesh=null
      SoftSourceActor=null
      SourceMode=AttachParent
      bAllowSamplingFromStreamingLODs=false
      bCaptureTransformsPerFrame=true
      bUsePhysicsBodyVelocity=false
      FilteredSockets []
      LODIndexUserParameter {parameter {name=None, typeDefHandle {registeredTypeIndex=108}, varData []}}
      MeshParameterBinding {parameter {name=None, typeDefHandle {registeredTypeIndex=1}, varData []}}
      SectionFilter {allowedMaterialSlots []}
     }}
   )
   &ShapeLocation(MaskSpawnGroup=false, Override Seed=false, Sphere Radius=5, UseOwnerScale=true)
  ]
  particleUpdateScript [
   &ParticleState()
   &Drag(
    Drag=&MaskFloatBySpawnGroup(
     Masked Float=&Multiply_Float(
      A=&RandomRangeFloat(Minimum=0.3, Maximum=0.5)
      B=&FloatFromCurve(
       FloatCurve=DataInterface() {propertyValues {Curve {keys [
           {interpMode=RCIM_Cubic, tangentMode=RCTM_Break, time=0.065, value=200}
           {interpMode=RCIM_Cubic, tangentMode=RCTM_Break, time=0.3385, value=1}
          ]}}}
       CurveIndex=$System.NormalizedLoopAge
      )
     )
     Passthrough Float=&RandomRangeFloat(Minimum=14, Maximum=16)
    )
    Use Linear Drag=true
    Use Rotational Drag=true
   )
   &ScaleVelocity(Velocity Scale=&MaskVectorBySpawnGroup(
     Masked Vector=&MakeVector(
      X=&FloatFromCurve(
       FloatCurve=DataInterface() {propertyValues {Curve {keys [{interpMode=RCIM_Cubic, tangentMode=RCTM_Break, time=0.06, value=4}, {interpMode=RCIM_Cubic, tangentMode=RCTM_Break, time=0.2, value=1}]}}}
       CurveIndex=$System.NormalizedLoopAge
      )
      Y=&FloatFromCurve(
       FloatCurve=DataInterface() {propertyValues {Curve {keys [{interpMode=RCIM_Cubic, tangentMode=RCTM_Break, time=0.3, value=0}, {interpMode=RCIM_Cubic, tangentMode=RCTM_Break, time=0.4, value=1}]}}}
       CurveIndex=$System.NormalizedLoopAge
      )
      Z=&FloatFromCurve(
       FloatCurve=DataInterface() {propertyValues {Curve {keys [{interpMode=RCIM_Cubic, tangentMode=RCTM_Break, time=0.06, value=4}, {interpMode=RCIM_Cubic, tangentMode=RCTM_Break, time=0.2, value=1}]}}}
       CurveIndex=$System.NormalizedLoopAge
      )
     )
     Passthrough Vector=<0.1 0.1 0.1>
    ))
   &GravityForce(Gravity=&MaskVectorBySpawnGroup(Masked Vector=&RandomRangeVector(Minimum=<0 0 2000>, Maximum=<0 0 5000>), Passthrough Vector=<0 0 0>))
   &CurlNoiseForce(
    Bias Noise=false
    Noise Frequency=100
    Noise Strength=&MaskFloatBySpawnGroup(
     Masked Float=&Multiply_Float(
      A=&RandomRangeFloat(Minimum=1000, Maximum=1300)
      B=&FloatFromCurve(
       FloatCurve=DataInterface() {propertyValues {Curve {keys [{interpMode=RCIM_Cubic, tangentMode=RCTM_Break, time=0.15, value=0}, {interpMode=RCIM_Cubic, tangentMode=RCTM_Break, time=0.3, value=1}]}}}
       CurveIndex=$System.NormalizedLoopAge
      )
     )
     Passthrough Float=0
    )
    Pan Noise=false
   )
   &SolveForcesAndVelocity(Clamp Velocity=false, Limit Acceleration=false)
   &SubUVAnimation(
    Random Start Frame=true
    "Sprite Renderer "=DataInterface() {propertyValues {SpriteRenderer=&self:Looping_Smoke_0.NiagaraSpriteRendererProperties_0}}
    Start Frame Offset=5
    UseEndFrame=false
    UseStartFrame=false
   )
   &SubUVAnimation(
    Random Start Frame=true
    "Sprite Renderer "=DataInterface() {propertyValues {SpriteRenderer=&self:NE_SmokeCones002_0.NiagaraSpriteRendererProperties_0}}
    SubUV Lookup Index Scale=0.5
    UseEndFrame=false
    UseStartFrame=false
   ) {module=SubUVAnimation001}
   &ScaleSpriteSize(
    Scale Sprite Size Mode=::NewEnumerator0
    Uniform Scale Factor=&MaskFloatBySpawnGroup(
     Masked Float=&FloatFromCurve(FloatCurve=DataInterface() {propertyValues {Curve {keys [
          {arriveTangent=0.5974, interpMode=RCIM_Cubic, leaveTangent=0.5974, tangentMode=RCTM_User, time=0, value=1}
          {arriveTangent=0.1871, interpMode=RCIM_Cubic, leaveTangent=0.1871, tangentMode=RCTM_User, time=1, value=1.5}
         ]}}})
     Passthrough Float=&FloatFromCurve(FloatCurve=DataInterface() {propertyValues {Curve {keys [
          {arriveTangent=3.6786, interpMode=RCIM_Cubic, leaveTangent=3.6786, tangentMode=RCTM_User, time=0, value=0}
          {interpMode=RCIM_Cubic, tangentMode=RCTM_User, time=1, value=1}
         ]}}})
    )
   )
   &ScaleColor(
    Scale Alpha=&MaskFloatBySpawnGroup(
     Masked Float=&FloatFromCurve(
      FloatCurve=DataInterface() {propertyValues {Curve {keys [{interpMode=RCIM_Cubic, tangentMode=RCTM_Break, time=0.5, value=1}, {interpMode=RCIM_Cubic, tangentMode=RCTM_Break, time=1, value=0}]}}}
      Scale Curve=&FNTimeOfDayScalar(Night Scale=0.4, Morning Scale=1, Evening Scale=1)
     )
     Passthrough Float=&FloatFromCurve(FloatCurve=DataInterface() {propertyValues {Curve {keys [{interpMode=RCIM_Cubic, tangentMode=RCTM_Break, time=0.6, value=1}, {interpMode=RCIM_Cubic, tangentMode=RCTM_Break, time=1, value=0}]}}})
    )
    ScaleA=true
    ScaleRGB=true
   )
  ]
  rendererClasses [&NiagaraSpriteRendererProperties]
  renderers [
   {rendererClass=&NiagaraSpriteRendererProperties}
   {rendererClass=&NiagaraSpriteRendererProperties, rendererIndex=1}
  ]
 }
 Looping_Smoke001 {
  emitterSpawnScript [/SetVariables(Emitter.SmokeDepthFade=10) {isSetParameters=true, module=SetVariables_#4}]
  emitterUpdateScript [
   &EmitterState(Delay First Loop Only=true, Life Cycle Mode=::NewEnumerator1, Loop Delay=0.4, Loop Duration=30, UseLoopDelay=true)
   &SpawnRate(
    SpawnRate=&Multiply_Float(
     A=70
     B=&FloatFromCurve(
      FloatCurve=DataInterface() {propertyValues {Curve {keys [{interpMode=RCIM_Cubic, tangentMode=RCTM_Break, time=0.025, value=2}, {interpMode=RCIM_Cubic, tangentMode=RCTM_Break, time=0.22, value=1}]}}}
      CurveIndex=$Emitter.NormalizedLoopAge
     )
    )
    Use Spawn Probability=false
   ) {module=SpawnRate001}
   &SpawnRate(SpawnGroup=1, SpawnRate=9, Use Spawn Probability=false)
  ]
  particleSpawnScript [
   /SetVariables(Particles.VisibilityTag=&SpawnGroup()) {isSetParameters=true, module=SetVariables_#2}
   &InitializeParticle(
    Color=&MaskLinearColorBySpawnGroup(Masked Linear Color=<rgba 0.4 0.4 0.4 0.15>, Passthrough Linear Color=<rgba 0.4 0.4 0.4 0.08>)
    Lifetime=&MaskFloatBySpawnGroup(
     Masked Float=&RandomRangeFloat(Minimum=2, Maximum=2.2)
     Passthrough Float=&RandomRangeFloat(Minimum=2.4, Maximum=2.8)
    )
    Position Mode=::NewEnumerator0
    Sprite Rotation Mode=::NewEnumerator2
    Sprite Size Mode=::NewEnumerator3
    Uniform Sprite Size=&MaskFloatBySpawnGroup(
     Masked Float=&RandomRangeFloat(Minimum=260, Maximum=340)
     Passthrough Float=&RandomRangeFloat(Minimum=320, Maximum=380)
    )
    Write Material Random=true
   )
   &DynamicMaterialParameters(
    Index 0 Param 1=1
    Module.Index0Float0Write=true
    Module.Index0Float1Write=true
    Module.Index0Float2Write=true
    Module.Index0Float3Write=true
   )
   &AddVelocity(Velocity=&MaskVectorBySpawnGroup(
     Masked Vector=&RandomRangeVector(Minimum=<400 -160 40>, Maximum=<600 160 70>)
     Passthrough Vector=<0 0 0>
    ))
   &AddVelocity(
    Cone Angle=360
    Cone Axis=<0 0 1>
    Module.UseConeFalloff=false
    Module.UseVelDistribution=false
    Override Seed=false
    Velocity Mode=::NewEnumerator2
    Velocity Speed=&MaskFloatBySpawnGroup(Masked Float=0, Passthrough Float=&RandomRangeFloat(Minimum=1600, Maximum=2000))
   ) {module=AddVelocity001}
   &StaticMeshLocation(
    MaskSpawnGroup=false
    Mesh Sampling Type:ENiagara_StaticSamplingMode=::NewEnumerator1
    SamplePosOffset=false
    Socket Index=1
    Socket Sampling Mode:ENiagara_StaticSocketMode=::NewEnumerator2
    Static Mesh=DataInterface() {propertyValues {
      DefaultMesh=null
      InstanceIndex=-1
      LODIndex=0
      PreviewMesh=null
      SoftSourceActor=null
      SourceMode=AttachParent
      bAllowSamplingFromStreamingLODs=false
      bCaptureTransformsPerFrame=true
      bUsePhysicsBodyVelocity=false
      FilteredSockets []
      LODIndexUserParameter {parameter {name=None, typeDefHandle {registeredTypeIndex=108}, varData []}}
      MeshParameterBinding {parameter {name=None, typeDefHandle {registeredTypeIndex=1}, varData []}}
      SectionFilter {allowedMaterialSlots []}
     }}
   )
   &ShapeLocation(MaskSpawnGroup=false, Override Seed=false, Sphere Radius=5, UseOwnerScale=true)
  ]
  particleUpdateScript [
   &ParticleState()
   &Drag(
    Drag=&MaskFloatBySpawnGroup(
     Masked Float=&Multiply_Float(
      A=&RandomRangeFloat(Minimum=0.3, Maximum=0.5)
      B=&FloatFromCurve(
       FloatCurve=DataInterface() {propertyValues {Curve {keys [
           {interpMode=RCIM_Cubic, tangentMode=RCTM_Break, time=0.065, value=200}
           {interpMode=RCIM_Cubic, tangentMode=RCTM_Break, time=0.3385, value=1}
          ]}}}
       CurveIndex=$System.NormalizedLoopAge
      )
     )
     Passthrough Float=&RandomRangeFloat(Minimum=14, Maximum=16)
    )
    Use Linear Drag=true
    Use Rotational Drag=true
   )
   &ScaleVelocity(Velocity Scale=&MaskVectorBySpawnGroup(
     Masked Vector=&MakeVector(
      X=&FloatFromCurve(
       FloatCurve=DataInterface() {propertyValues {Curve {keys [{interpMode=RCIM_Cubic, tangentMode=RCTM_Break, time=0.06, value=4}, {interpMode=RCIM_Cubic, tangentMode=RCTM_Break, time=0.2, value=1}]}}}
       CurveIndex=$System.NormalizedLoopAge
      )
      Y=&FloatFromCurve(
       FloatCurve=DataInterface() {propertyValues {Curve {keys [{interpMode=RCIM_Cubic, tangentMode=RCTM_Break, time=0.3, value=0}, {interpMode=RCIM_Cubic, tangentMode=RCTM_Break, time=0.4, value=1}]}}}
       CurveIndex=$System.NormalizedLoopAge
      )
      Z=&FloatFromCurve(
       FloatCurve=DataInterface() {propertyValues {Curve {keys [{interpMode=RCIM_Cubic, tangentMode=RCTM_Break, time=0.06, value=4}, {interpMode=RCIM_Cubic, tangentMode=RCTM_Break, time=0.2, value=1}]}}}
       CurveIndex=$System.NormalizedLoopAge
      )
     )
     Passthrough Vector=<0.1 0.1 0.1>
    ))
   &GravityForce(Gravity=&MaskVectorBySpawnGroup(Masked Vector=&RandomRangeVector(Minimum=<0 0 2000>, Maximum=<0 0 5000>), Passthrough Vector=<0 0 0>))
   &CurlNoiseForce(
    Bias Noise=false
    Noise Frequency=100
    Noise Strength=&MaskFloatBySpawnGroup(
     Masked Float=&Multiply_Float(
      A=&RandomRangeFloat(Minimum=1000, Maximum=1300)
      B=&FloatFromCurve(
       FloatCurve=DataInterface() {propertyValues {Curve {keys [{interpMode=RCIM_Cubic, tangentMode=RCTM_Break, time=0.15, value=0}, {interpMode=RCIM_Cubic, tangentMode=RCTM_Break, time=0.3, value=1}]}}}
       CurveIndex=$System.NormalizedLoopAge
      )
     )
     Passthrough Float=0
    )
    Pan Noise=false
   )
   &SolveForcesAndVelocity(Clamp Velocity=false, Limit Acceleration=false)
   &SubUVAnimation(
    Random Start Frame=true
    "Sprite Renderer "=DataInterface() {propertyValues {SpriteRenderer=&self:Looping_Smoke001_0.NiagaraSpriteRendererProperties_0}}
    Start Frame Offset=5
    UseEndFrame=false
    UseStartFrame=false
   )
   &SubUVAnimation(
    Random Start Frame=true
    "Sprite Renderer "=DataInterface() {propertyValues {SpriteRenderer=&self:NE_SmokeCones002_0.NiagaraSpriteRendererProperties_0}}
    SubUV Lookup Index Scale=0.5
    UseEndFrame=false
    UseStartFrame=false
   ) {module=SubUVAnimation001}
   &ScaleSpriteSize(
    Scale Sprite Size Mode=::NewEnumerator0
    Uniform Scale Factor=&MaskFloatBySpawnGroup(
     Masked Float=&FloatFromCurve(FloatCurve=DataInterface() {propertyValues {Curve {keys [
          {arriveTangent=0.5974, interpMode=RCIM_Cubic, leaveTangent=0.5974, tangentMode=RCTM_User, time=0, value=1}
          {arriveTangent=0.1871, interpMode=RCIM_Cubic, leaveTangent=0.1871, tangentMode=RCTM_User, time=1, value=1.5}
         ]}}})
     Passthrough Float=&FloatFromCurve(FloatCurve=DataInterface() {propertyValues {Curve {keys [
          {arriveTangent=3.6786, interpMode=RCIM_Cubic, leaveTangent=3.6786, tangentMode=RCTM_User, time=0, value=0}
          {interpMode=RCIM_Cubic, tangentMode=RCTM_User, time=1, value=1}
         ]}}})
    )
   )
   &ScaleColor(
    Scale Alpha=&MaskFloatBySpawnGroup(
     Masked Float=&FloatFromCurve(
      FloatCurve=DataInterface() {propertyValues {Curve {keys [{interpMode=RCIM_Cubic, tangentMode=RCTM_Break, time=0.5, value=1}, {interpMode=RCIM_Cubic, tangentMode=RCTM_Break, time=1, value=0}]}}}
      Scale Curve=&FNTimeOfDayScalar(Night Scale=0.4, Morning Scale=1, Evening Scale=1)
     )
     Passthrough Float=&FloatFromCurve(FloatCurve=DataInterface() {propertyValues {Curve {keys [{interpMode=RCIM_Cubic, tangentMode=RCTM_Break, time=0.6, value=1}, {interpMode=RCIM_Cubic, tangentMode=RCTM_Break, time=1, value=0}]}}})
    )
    ScaleA=true
    ScaleRGB=true
   )
  ]
  rendererClasses [&NiagaraSpriteRendererProperties]
  renderers [
   {rendererClass=&NiagaraSpriteRendererProperties}
   {rendererClass=&NiagaraSpriteRendererProperties, rendererIndex=1}
  ]
 }
 Start_HotSmoke {
  emitterUpdateScript [
   &EmitterState(Life Cycle Mode=::NewEnumerator1, Loop Behavior=::NewEnumerator1, UseLoopDelay=false)
   &SpawnRate(
    SpawnRate=&Multiply_Float(
     A=15
     B=&FloatFromCurve(
      FloatCurve=DataInterface() {propertyValues {Curve {keys [
          {arriveTangent=17.1995, interpMode=RCIM_Cubic, leaveTangent=17.1995, tangentMode=RCTM_User, time=0.0547, value=0}
          {arriveTangent=0.0806, interpMode=RCIM_Cubic, leaveTangent=0.0806, tangentMode=RCTM_User, time=0.12, value=1}
          {interpMode=RCIM_Cubic, tangentMode=RCTM_Break, time=0.65, value=1}
          {interpMode=RCIM_Cubic, tangentMode=RCTM_Break, time=1, value=0}
         ]}}}
      CurveIndex=$Emitter.NormalizedLoopAge
     )
    )
    Use Spawn Probability=false
   )
   &SpawnRate(
    SpawnGroup=1
    SpawnRate=&Multiply_Float(
     A=15
     B=&FloatFromCurve(
      FloatCurve=DataInterface() {propertyValues {Curve {keys [
          {arriveTangent=17.1995, interpMode=RCIM_Cubic, leaveTangent=17.1995, tangentMode=RCTM_User, time=0.0547, value=0}
          {arriveTangent=0.0806, interpMode=RCIM_Cubic, leaveTangent=0.0806, tangentMode=RCTM_User, time=0.12, value=1}
          {interpMode=RCIM_Cubic, tangentMode=RCTM_Break, time=0.65, value=1}
          {interpMode=RCIM_Cubic, tangentMode=RCTM_Break, time=1, value=0}
         ]}}}
      CurveIndex=$Emitter.NormalizedLoopAge
     )
    )
    Use Spawn Probability=false
   ) {module=SpawnRate001}
   &SpawnRate(
    SpawnGroup=2
    SpawnRate=&Multiply_Float(
     A=15
     B=&FloatFromCurve(
      FloatCurve=DataInterface() {propertyValues {Curve {keys [
          {arriveTangent=17.1995, interpMode=RCIM_Cubic, leaveTangent=17.1995, tangentMode=RCTM_User, time=0.0547, value=0}
          {arriveTangent=0.0806, interpMode=RCIM_Cubic, leaveTangent=0.0806, tangentMode=RCTM_User, time=0.12, value=1}
          {interpMode=RCIM_Cubic, tangentMode=RCTM_Break, time=0.65, value=1}
          {interpMode=RCIM_Cubic, tangentMode=RCTM_Break, time=1, value=0}
         ]}}}
      CurveIndex=$Emitter.NormalizedLoopAge
     )
    )
    Use Spawn Probability=false
   ) {module=SpawnRate003}
   &SpawnRate(
    SpawnGroup=3
    SpawnRate=&Multiply_Float(
     A=20
     B=&FloatFromCurve(
      FloatCurve=DataInterface() {propertyValues {Curve {keys [
          <at 0.087 0 RCIM_Cubic>
          {arriveTangent=0.194, interpMode=RCIM_Cubic, leaveTangent=0.194, tangentMode=RCTM_User, time=0.15, value=1}
          {interpMode=RCIM_Cubic, tangentMode=RCTM_Break, time=0.5, value=1}
          {interpMode=RCIM_Cubic, tangentMode=RCTM_Break, time=0.75, value=0}
         ]}}}
      CurveIndex=$Emitter.NormalizedLoopAge
     )
    )
    Use Spawn Probability=false
   ) {module=SpawnRate002}
  ]
  particleSpawnScript [
   /SetVariables(Particles.VisibilityTag=&SpawnGroup()) {isSetParameters=true, module=SetVariables_#2}
   &PartitionParticles(Number Of Partitions=3)
   &InitializeParticle(
    Color=&MaskLinearColorBySpawnGroup(
     Masked Linear Color=&RandomRangeLinearColor(Minimum=<rgba 0.339 0.339 0.339 0.85>, Maximum=<rgba 0.432 0.432 0.432 1>)
     Passthrough Linear Color=&MaskLinearColorBySpawnGroup(
      Masked Linear Color=&RandomRangeLinearColor(Minimum=<rgba 0.339 0.339 0.339 0.85>, Maximum=<rgba 0.432 0.432 0.432 1>)
      Passthrough Linear Color=&MaskLinearColorBySpawnGroup(
       Masked Linear Color=&RandomRangeLinearColor(Minimum=<rgba 0.339 0.339 0.339 0.85>, Maximum=<rgba 0.432 0.432 0.432 1>)
       Passthrough Linear Color=<rgba 0.4 0.4 0.4 0.075>
       Spawn Group Mask=2
      )
      Spawn Group Mask=1
     )
    )
    Lifetime=&MaskFloatBySpawnGroup(
     Masked Float=&RandomRangeFloat(Minimum=0.1, Maximum=0.2)
     Passthrough Float=&MaskFloatBySpawnGroup(
      Masked Float=&RandomRangeFloat(Minimum=0.1, Maximum=0.2)
      Passthrough Float=&MaskFloatBySpawnGroup(
       Masked Float=&RandomRangeFloat(Minimum=0.1, Maximum=0.2)
       Passthrough Float=&RandomRangeFloat(Minimum=2.2, Maximum=2.6)
       Spawn Group Mask=2
      )
      Spawn Group Mask=1
     )
    )
    Mesh Scale Max=<1.4 1.4 1.4>
    Mesh Scale Mode=::NewEnumerator2
    Position Mode=::NewEnumerator1
    Position Offset=&MaskVectorBySpawnGroup(
     Masked Vector=<5 0 0>
     Passthrough Vector=&MaskVectorBySpawnGroup(Masked Vector=<-5 0 0>, Passthrough Vector=<0 0 6>, Spawn Group Mask=1)
    )
    Sprite Rotation Mode=::NewEnumerator2
    Sprite Size Mode=::NewEnumerator1
    Uniform Sprite Size Max=640
    Uniform Sprite Size Min=492
    UsePositionOffset=true
    Write Material Random=true
   )
   &InitialMeshOrientation(
    Mesh Orientation Mode=::NewEnumerator7
    Rotation=&MaskVectorBySpawnGroup(
     Masked Vector=<0 -0.5 0>
     Passthrough Vector=&MaskVectorBySpawnGroup(Masked Vector=<0 0 -0.25>, Passthrough Vector=<0 0 0.25>, Spawn Group Mask=1)
    )
    Rotation Coordinate Space=::NewEnumerator2
    Use Rotation Vector=true
   )
   &AddVelocity(
    Cone Angle=0
    Cone Axis=&SelectVectorFromArray(
     Direct Array Index=$Particles.PartitionParticles.Partition
     Vector Selection Array=DataInterface() {propertyValues {FloatData [<0 0 1>, <0 -0.25 0>, <0 0.25 0>]}}
    )
    Module.UseConeFalloff=false
    Module.UseVelDistribution=false
    Override Seed=false
    Velocity Mode=::NewEnumerator2
    Velocity Speed=&MaskFloatBySpawnGroup(
     Masked Float=0
     Passthrough Float=&MaskFloatBySpawnGroup(
      Masked Float=0
      Passthrough Float=&MaskFloatBySpawnGroup(Masked Float=0, Passthrough Float=&RandomRangeFloat(Minimum=1500, Maximum=4000), Spawn Group Mask=2)
      Spawn Group Mask=1
     )
    )
   )
   &StaticMeshLocation(
    MaskSpawnGroup=false
    Mesh Sampling Type:ENiagara_StaticSamplingMode=::NewEnumerator1
    SamplePosOffset=false
    Socket Sampling Mode:ENiagara_StaticSocketMode=::NewEnumerator2
    Static Mesh=DataInterface() {propertyValues {
      DefaultMesh=null
      InstanceIndex=-1
      LODIndex=0
      PreviewMesh=null
      SoftSourceActor=null
      SourceMode=AttachParent
      bAllowSamplingFromStreamingLODs=false
      bCaptureTransformsPerFrame=true
      bUsePhysicsBodyVelocity=false
      FilteredSockets []
      LODIndexUserParameter {parameter {name=None, typeDefHandle {registeredTypeIndex=108}, varData []}}
      MeshParameterBinding {parameter {name=None, typeDefHandle {registeredTypeIndex=1}, varData []}}
      SectionFilter {allowedMaterialSlots []}
     }}
   )
   &ApplyInitialForces()
  ]
  particleUpdateScript [
   &ParticleState()
   &ScaleMeshSize(Scale Factor=&Multiply_VectorByFloat(
     Vector=&MakeVector(
      X=&FloatFromCurve(FloatCurve=DataInterface() {propertyValues {Curve {keys [<at 0 0.5>,<at 1 1.5>]}}})
      Y=&FloatFromCurve(FloatCurve=DataInterface() {propertyValues {Curve {keys [<at 0 0.5>,<at 1 1.5>]}}})
      Z=&FloatFromCurve(FloatCurve=DataInterface() {propertyValues {Curve {keys [<at 0 0.5>,<at 1 1.5>]}}})
     )
     Float=3
    ))
   &Drag(Drag=&RandomRangeFloat(Minimum=10, Maximum=12), Use Linear Drag=true, Use Rotational Drag=true)
   /SetVariables(Particles.Smoke_DynamicParameter:Vector4f=&MakeVector4(
     X=&FloatFromCurve(FloatCurve=DataInterface() {propertyValues {Curve {keys [
          {arriveTangent=3.0057, interpMode=RCIM_Cubic, leaveTangent=3.0057, tangentMode=RCTM_User, time=0, value=0}
          {arriveTangent=0.5254, interpMode=RCIM_Cubic, leaveTangent=0.5254, tangentMode=RCTM_User, time=1, value=1}
         ]}}})
     Y=0
     Z=0
     W=0
    )) {isSetParameters=true, module=SetVariables_#5}
   &ScaleSpriteSize(Uniform Curve Sprite Scale=DataInterface() {propertyValues {Curve {keys [
        {arriveTangent=3.0506, interpMode=RCIM_Cubic, leaveTangent=3.0506, tangentMode=RCTM_User, time=0, value=0}
        {interpMode=RCIM_Cubic, tangentMode=RCTM_User, time=1, value=1}
       ]}}})
   &DynamicMaterialParameters(
    Index 0 Param 3=1
    Index 0 Param 4=1
    Module.Index0Param0Write=true
    Module.Index0Param1Write=true
    Module.Index0Param2Write=true
    Module.Index0Param3Write=true
   )
   &ScaleColor(
    Scale Alpha=&MaskFloatBySpawnGroup(
     Masked Float=&FloatFromCurve(
      FloatCurve=DataInterface() {propertyValues {Curve {keys [
          {interpMode=RCIM_Cubic, leaveTangent=13.0239, tangentMode=RCTM_Break, time=0, value=0}
          {interpMode=RCIM_Cubic, leaveTangent=-2.1575, tangentMode=RCTM_Break, time=0.1, value=1}
          {interpMode=RCIM_Cubic, tangentMode=RCTM_Break, time=1, value=0}
         ]}}}
      Scale Curve=&FNTimeOfDayScalar(Night Scale=0.4, Morning Scale=1, Evening Scale=1)
     )
     Passthrough Float=&MaskFloatBySpawnGroup(
      Masked Float=&FloatFromCurve(
       FloatCurve=DataInterface() {propertyValues {Curve {keys [{interpMode=RCIM_Cubic, tangentMode=RCTM_Break, time=0.6, value=1}, {interpMode=RCIM_Cubic, tangentMode=RCTM_Break, time=1, value=0}]}}}
       Scale Curve=&FNTimeOfDayScalar(Night Scale=0.4, Morning Scale=1, Evening Scale=1)
      )
      Passthrough Float=&MaskFloatBySpawnGroup(
       Masked Float=&FloatFromCurve(
        FloatCurve=DataInterface() {propertyValues {Curve {keys [{interpMode=RCIM_Cubic, tangentMode=RCTM_Break, time=0.6, value=1}, {interpMode=RCIM_Cubic, tangentMode=RCTM_Break, time=1, value=0}]}}}
        Scale Curve=&FNTimeOfDayScalar(Night Scale=0.4, Morning Scale=1, Evening Scale=1)
       )
       Passthrough Float=&FloatFromCurve(FloatCurve=DataInterface() {propertyValues {Curve {keys [{interpMode=RCIM_Cubic, tangentMode=RCTM_Break, time=0.6, value=1}, {interpMode=RCIM_Cubic, tangentMode=RCTM_Break, time=1, value=0}]}}})
       Spawn Group Mask=2
      )
      Spawn Group Mask=1
     )
    )
    ScaleA=true
    ScaleRGB=true
   )
   &ScaleColor(
    Scale Alpha=&MaskFloatBySpawnGroup(
     Masked Float=&FloatFromCurve(
      FloatCurve=DataInterface() {propertyValues {Curve {keys [
          <at 0 0 RCIM_Cubic>
          {arriveTangent=0.0888, interpMode=RCIM_Cubic, leaveTangent=0.0888, tangentMode=RCTM_User, time=0.2, value=1}
          {arriveTangent=-0.0921, interpMode=RCIM_Cubic, leaveTangent=-0.0921, tangentMode=RCTM_User, time=0.65, value=1}
          {interpMode=RCIM_Cubic, tangentMode=RCTM_Break, time=1, value=0}
          {interpMode=RCIM_Cubic, tangentMode=RCTM_Break, time=1, value=0}
         ]}}}
      CurveIndex=$Emitter.NormalizedLoopAge
     )
     Passthrough Float=&MaskFloatBySpawnGroup(
      Masked Float=&FloatFromCurve(
       FloatCurve=DataInterface() {propertyValues {Curve {keys [
           <at 0 0 RCIM_Cubic>
           {arriveTangent=0.0888, interpMode=RCIM_Cubic, leaveTangent=0.0888, tangentMode=RCTM_User, time=0.2, value=1}
           {arriveTangent=-0.0921, interpMode=RCIM_Cubic, leaveTangent=-0.0921, tangentMode=RCTM_User, time=0.65, value=1}
           {interpMode=RCIM_Cubic, tangentMode=RCTM_Break, time=1, value=0}
           {interpMode=RCIM_Cubic, tangentMode=RCTM_Break, time=1, value=0}
          ]}}}
       CurveIndex=$Emitter.NormalizedLoopAge
      )
      Passthrough Float=&MaskFloatBySpawnGroup(
       Masked Float=&FloatFromCurve(
        FloatCurve=DataInterface() {propertyValues {Curve {keys [
            <at 0 0 RCIM_Cubic>
            {arriveTangent=0.0888, interpMode=RCIM_Cubic, leaveTangent=0.0888, tangentMode=RCTM_User, time=0.2, value=1}
            {arriveTangent=-0.0921, interpMode=RCIM_Cubic, leaveTangent=-0.0921, tangentMode=RCTM_User, time=0.65, value=1}
            {interpMode=RCIM_Cubic, tangentMode=RCTM_Break, time=1, value=0}
            {interpMode=RCIM_Cubic, tangentMode=RCTM_Break, time=1, value=0}
           ]}}}
        CurveIndex=$Emitter.NormalizedLoopAge
       )
       Passthrough Float=&FloatFromCurve(
        FloatCurve=DataInterface() {propertyValues {Curve {keys [{interpMode=RCIM_Cubic, tangentMode=RCTM_Break, time=0.6, value=1}, {interpMode=RCIM_Cubic, tangentMode=RCTM_Break, time=0.9, value=0}]}}}
        CurveIndex=$Emitter.NormalizedLoopAge
       )
       Spawn Group Mask=2
      )
      Spawn Group Mask=1
     )
    )
    ScaleA=true
    ScaleRGB=true
   ) {module=ScaleColor001}
   &SubUVAnimation(
    Random Start Frame=true
    "Sprite Renderer "=DataInterface() {propertyValues {SpriteRenderer=&self:NE_SmokeCones001_4.NiagaraSpriteRendererProperties_0}}
    SubUV Lookup Index Scale=0.5
    UseEndFrame=false
    UseStartFrame=false
   )
   &SolveForcesAndVelocity(Clamp Velocity=false, Limit Acceleration=false)
  ]
  rendererClasses [&NiagaraMeshRendererProperties, &NiagaraSpriteRendererProperties]
  renderers [
   {rendererClass=&NiagaraMeshRendererProperties, rendererIndex=0}
   {rendererClass=&NiagaraMeshRendererProperties, rendererIndex=1}
   {rendererClass=&NiagaraMeshRendererProperties, rendererIndex=2}
   {rendererClass=&NiagaraSpriteRendererProperties, rendererIndex=3}
  ]
 }
 Start_HotSmoke001 {
  emitterUpdateScript [
   &EmitterState(Life Cycle Mode=::NewEnumerator1, Loop Behavior=::NewEnumerator1, UseLoopDelay=false)
   &SpawnRate(
    SpawnRate=&Multiply_Float(
     A=15
     B=&FloatFromCurve(
      FloatCurve=DataInterface() {propertyValues {Curve {keys [
          {arriveTangent=17.1995, interpMode=RCIM_Cubic, leaveTangent=17.1995, tangentMode=RCTM_User, time=0.0547, value=0}
          {arriveTangent=0.0806, interpMode=RCIM_Cubic, leaveTangent=0.0806, tangentMode=RCTM_User, time=0.12, value=1}
          {interpMode=RCIM_Cubic, tangentMode=RCTM_Break, time=0.65, value=1}
          {interpMode=RCIM_Cubic, tangentMode=RCTM_Break, time=1, value=0}
         ]}}}
      CurveIndex=$Emitter.NormalizedLoopAge
     )
    )
    Use Spawn Probability=false
   )
   &SpawnRate(
    SpawnGroup=1
    SpawnRate=&Multiply_Float(
     A=15
     B=&FloatFromCurve(
      FloatCurve=DataInterface() {propertyValues {Curve {keys [
          {arriveTangent=17.1995, interpMode=RCIM_Cubic, leaveTangent=17.1995, tangentMode=RCTM_User, time=0.0547, value=0}
          {arriveTangent=0.0806, interpMode=RCIM_Cubic, leaveTangent=0.0806, tangentMode=RCTM_User, time=0.12, value=1}
          {interpMode=RCIM_Cubic, tangentMode=RCTM_Break, time=0.65, value=1}
          {interpMode=RCIM_Cubic, tangentMode=RCTM_Break, time=1, value=0}
         ]}}}
      CurveIndex=$Emitter.NormalizedLoopAge
     )
    )
    Use Spawn Probability=false
   ) {module=SpawnRate001}
   &SpawnRate(
    SpawnGroup=2
    SpawnRate=&Multiply_Float(
     A=15
     B=&FloatFromCurve(
      FloatCurve=DataInterface() {propertyValues {Curve {keys [
          {arriveTangent=17.1995, interpMode=RCIM_Cubic, leaveTangent=17.1995, tangentMode=RCTM_User, time=0.0547, value=0}
          {arriveTangent=0.0806, interpMode=RCIM_Cubic, leaveTangent=0.0806, tangentMode=RCTM_User, time=0.12, value=1}
          {interpMode=RCIM_Cubic, tangentMode=RCTM_Break, time=0.65, value=1}
          {interpMode=RCIM_Cubic, tangentMode=RCTM_Break, time=1, value=0}
         ]}}}
      CurveIndex=$Emitter.NormalizedLoopAge
     )
    )
    Use Spawn Probability=false
   ) {module=SpawnRate003}
   &SpawnRate(
    SpawnGroup=3
    SpawnRate=&Multiply_Float(
     A=20
     B=&FloatFromCurve(
      FloatCurve=DataInterface() {propertyValues {Curve {keys [
          <at 0.087 0 RCIM_Cubic>
          {arriveTangent=0.194, interpMode=RCIM_Cubic, leaveTangent=0.194, tangentMode=RCTM_User, time=0.15, value=1}
          {interpMode=RCIM_Cubic, tangentMode=RCTM_Break, time=0.5, value=1}
          {interpMode=RCIM_Cubic, tangentMode=RCTM_Break, time=0.75, value=0}
         ]}}}
      CurveIndex=$Emitter.NormalizedLoopAge
     )
    )
    Use Spawn Probability=false
   ) {module=SpawnRate002}
  ]
  particleSpawnScript [
   /SetVariables(Particles.VisibilityTag=&SpawnGroup()) {isSetParameters=true, module=SetVariables_#2}
   &PartitionParticles(Number Of Partitions=3)
   &InitializeParticle(
    Color=&MaskLinearColorBySpawnGroup(
     Masked Linear Color=&RandomRangeLinearColor(Minimum=<rgba 0.339 0.339 0.339 0.85>, Maximum=<rgba 0.432 0.432 0.432 1>)
     Passthrough Linear Color=&MaskLinearColorBySpawnGroup(
      Masked Linear Color=&RandomRangeLinearColor(Minimum=<rgba 0.339 0.339 0.339 0.85>, Maximum=<rgba 0.432 0.432 0.432 1>)
      Passthrough Linear Color=&MaskLinearColorBySpawnGroup(
       Masked Linear Color=&RandomRangeLinearColor(Minimum=<rgba 0.339 0.339 0.339 0.85>, Maximum=<rgba 0.432 0.432 0.432 1>)
       Passthrough Linear Color=<rgba 0.4 0.4 0.4 0.075>
       Spawn Group Mask=2
      )
      Spawn Group Mask=1
     )
    )
    Lifetime=&MaskFloatBySpawnGroup(
     Masked Float=&RandomRangeFloat(Minimum=0.1, Maximum=0.2)
     Passthrough Float=&MaskFloatBySpawnGroup(
      Masked Float=&RandomRangeFloat(Minimum=0.1, Maximum=0.2)
      Passthrough Float=&MaskFloatBySpawnGroup(
       Masked Float=&RandomRangeFloat(Minimum=0.1, Maximum=0.2)
       Passthrough Float=&RandomRangeFloat(Minimum=2.2, Maximum=2.6)
       Spawn Group Mask=2
      )
      Spawn Group Mask=1
     )
    )
    Mesh Scale Max=<1.4 1.4 1.4>
    Mesh Scale Mode=::NewEnumerator2
    Position Mode=::NewEnumerator1
    Position Offset=&MaskVectorBySpawnGroup(
     Masked Vector=<5 0 0>
     Passthrough Vector=&MaskVectorBySpawnGroup(Masked Vector=<-5 0 0>, Passthrough Vector=<0 0 6>, Spawn Group Mask=1)
    )
    Sprite Rotation Mode=::NewEnumerator2
    Sprite Size Mode=::NewEnumerator1
    Uniform Sprite Size Max=640
    Uniform Sprite Size Min=492
    UsePositionOffset=true
    Write Material Random=true
   )
   &InitialMeshOrientation(
    Mesh Orientation Mode=::NewEnumerator7
    Rotation=&MaskVectorBySpawnGroup(
     Masked Vector=<0 -0.5 0>
     Passthrough Vector=&MaskVectorBySpawnGroup(Masked Vector=<0 0 -0.25>, Passthrough Vector=<0 0 0.25>, Spawn Group Mask=1)
    )
    Rotation Coordinate Space=::NewEnumerator2
    Use Rotation Vector=true
   )
   &AddVelocity(
    Cone Angle=0
    Cone Axis=&SelectVectorFromArray(
     Direct Array Index=$Particles.PartitionParticles.Partition
     Vector Selection Array=DataInterface() {propertyValues {FloatData [<0 0 1>, <0 -0.25 0>, <0 0.25 0>]}}
    )
    Module.UseConeFalloff=false
    Module.UseVelDistribution=false
    Override Seed=false
    Velocity Mode=::NewEnumerator2
    Velocity Speed=&MaskFloatBySpawnGroup(
     Masked Float=0
     Passthrough Float=&MaskFloatBySpawnGroup(
      Masked Float=0
      Passthrough Float=&MaskFloatBySpawnGroup(Masked Float=0, Passthrough Float=&RandomRangeFloat(Minimum=1500, Maximum=4000), Spawn Group Mask=2)
      Spawn Group Mask=1
     )
    )
   )
   &StaticMeshLocation(
    MaskSpawnGroup=false
    Mesh Sampling Type:ENiagara_StaticSamplingMode=::NewEnumerator1
    SamplePosOffset=false
    Socket Index=1
    Socket Sampling Mode:ENiagara_StaticSocketMode=::NewEnumerator2
    Static Mesh=DataInterface() {propertyValues {
      DefaultMesh=null
      InstanceIndex=-1
      LODIndex=0
      PreviewMesh=null
      SoftSourceActor=null
      SourceMode=AttachParent
      bAllowSamplingFromStreamingLODs=false
      bCaptureTransformsPerFrame=true
      bUsePhysicsBodyVelocity=false
      FilteredSockets []
      LODIndexUserParameter {parameter {name=None, typeDefHandle {registeredTypeIndex=108}, varData []}}
      MeshParameterBinding {parameter {name=None, typeDefHandle {registeredTypeIndex=1}, varData []}}
      SectionFilter {allowedMaterialSlots []}
     }}
   )
   &ApplyInitialForces()
  ]
  particleUpdateScript [
   &ParticleState()
   &ScaleMeshSize(Scale Factor=&Multiply_VectorByFloat(
     Vector=&MakeVector(
      X=&FloatFromCurve(FloatCurve=DataInterface() {propertyValues {Curve {keys [<at 0 0.5>,<at 1 1.5>]}}})
      Y=&FloatFromCurve(FloatCurve=DataInterface() {propertyValues {Curve {keys [<at 0 0.5>,<at 1 1.5>]}}})
      Z=&FloatFromCurve(FloatCurve=DataInterface() {propertyValues {Curve {keys [<at 0 0.5>,<at 1 1.5>]}}})
     )
     Float=3
    ))
   &Drag(Drag=&RandomRangeFloat(Minimum=10, Maximum=12), Use Linear Drag=true, Use Rotational Drag=true)
   /SetVariables(Particles.Smoke_DynamicParameter:Vector4f=&MakeVector4(
     X=&FloatFromCurve(FloatCurve=DataInterface() {propertyValues {Curve {keys [
          {arriveTangent=3.0057, interpMode=RCIM_Cubic, leaveTangent=3.0057, tangentMode=RCTM_User, time=0, value=0}
          {arriveTangent=0.5254, interpMode=RCIM_Cubic, leaveTangent=0.5254, tangentMode=RCTM_User, time=1, value=1}
         ]}}})
     Y=0
     Z=0
     W=0
    )) {isSetParameters=true, module=SetVariables_#5}
   &ScaleSpriteSize(Uniform Curve Sprite Scale=DataInterface() {propertyValues {Curve {keys [
        {arriveTangent=3.0506, interpMode=RCIM_Cubic, leaveTangent=3.0506, tangentMode=RCTM_User, time=0, value=0}
        {interpMode=RCIM_Cubic, tangentMode=RCTM_User, time=1, value=1}
       ]}}})
   &DynamicMaterialParameters(
    Index 0 Param 3=1
    Index 0 Param 4=1
    Module.Index0Param0Write=true
    Module.Index0Param1Write=true
    Module.Index0Param2Write=true
    Module.Index0Param3Write=true
   )
   &ScaleColor(
    Scale Alpha=&MaskFloatBySpawnGroup(
     Masked Float=&FloatFromCurve(
      FloatCurve=DataInterface() {propertyValues {Curve {keys [
          {interpMode=RCIM_Cubic, leaveTangent=13.0239, tangentMode=RCTM_Break, time=0, value=0}
          {interpMode=RCIM_Cubic, leaveTangent=-2.1575, tangentMode=RCTM_Break, time=0.1, value=1}
          {interpMode=RCIM_Cubic, tangentMode=RCTM_Break, time=1, value=0}
         ]}}}
      Scale Curve=&FNTimeOfDayScalar(Night Scale=0.4, Morning Scale=1, Evening Scale=1)
     )
     Passthrough Float=&MaskFloatBySpawnGroup(
      Masked Float=&FloatFromCurve(
       FloatCurve=DataInterface() {propertyValues {Curve {keys [{interpMode=RCIM_Cubic, tangentMode=RCTM_Break, time=0.6, value=1}, {interpMode=RCIM_Cubic, tangentMode=RCTM_Break, time=1, value=0}]}}}
       Scale Curve=&FNTimeOfDayScalar(Night Scale=0.4, Morning Scale=1, Evening Scale=1)
      )
      Passthrough Float=&MaskFloatBySpawnGroup(
       Masked Float=&FloatFromCurve(
        FloatCurve=DataInterface() {propertyValues {Curve {keys [{interpMode=RCIM_Cubic, tangentMode=RCTM_Break, time=0.6, value=1}, {interpMode=RCIM_Cubic, tangentMode=RCTM_Break, time=1, value=0}]}}}
        Scale Curve=&FNTimeOfDayScalar(Night Scale=0.4, Morning Scale=1, Evening Scale=1)
       )
       Passthrough Float=&FloatFromCurve(FloatCurve=DataInterface() {propertyValues {Curve {keys [{interpMode=RCIM_Cubic, tangentMode=RCTM_Break, time=0.6, value=1}, {interpMode=RCIM_Cubic, tangentMode=RCTM_Break, time=1, value=0}]}}})
       Spawn Group Mask=2
      )
      Spawn Group Mask=1
     )
    )
    ScaleA=true
    ScaleRGB=true
   )
   &ScaleColor(
    Scale Alpha=&MaskFloatBySpawnGroup(
     Masked Float=&FloatFromCurve(
      FloatCurve=DataInterface() {propertyValues {Curve {keys [
          <at 0 0 RCIM_Cubic>
          {arriveTangent=0.0888, interpMode=RCIM_Cubic, leaveTangent=0.0888, tangentMode=RCTM_User, time=0.2, value=1}
          {arriveTangent=-0.0921, interpMode=RCIM_Cubic, leaveTangent=-0.0921, tangentMode=RCTM_User, time=0.65, value=1}
          {interpMode=RCIM_Cubic, tangentMode=RCTM_Break, time=1, value=0}
          {interpMode=RCIM_Cubic, tangentMode=RCTM_Break, time=1, value=0}
         ]}}}
      CurveIndex=$Emitter.NormalizedLoopAge
     )
     Passthrough Float=&MaskFloatBySpawnGroup(
      Masked Float=&FloatFromCurve(
       FloatCurve=DataInterface() {propertyValues {Curve {keys [
           <at 0 0 RCIM_Cubic>
           {arriveTangent=0.0888, interpMode=RCIM_Cubic, leaveTangent=0.0888, tangentMode=RCTM_User, time=0.2, value=1}
           {arriveTangent=-0.0921, interpMode=RCIM_Cubic, leaveTangent=-0.0921, tangentMode=RCTM_User, time=0.65, value=1}
           {interpMode=RCIM_Cubic, tangentMode=RCTM_Break, time=1, value=0}
           {interpMode=RCIM_Cubic, tangentMode=RCTM_Break, time=1, value=0}
          ]}}}
       CurveIndex=$Emitter.NormalizedLoopAge
      )
      Passthrough Float=&MaskFloatBySpawnGroup(
       Masked Float=&FloatFromCurve(
        FloatCurve=DataInterface() {propertyValues {Curve {keys [
            <at 0 0 RCIM_Cubic>
            {arriveTangent=0.0888, interpMode=RCIM_Cubic, leaveTangent=0.0888, tangentMode=RCTM_User, time=0.2, value=1}
            {arriveTangent=-0.0921, interpMode=RCIM_Cubic, leaveTangent=-0.0921, tangentMode=RCTM_User, time=0.65, value=1}
            {interpMode=RCIM_Cubic, tangentMode=RCTM_Break, time=1, value=0}
            {interpMode=RCIM_Cubic, tangentMode=RCTM_Break, time=1, value=0}
           ]}}}
        CurveIndex=$Emitter.NormalizedLoopAge
       )
       Passthrough Float=&FloatFromCurve(
        FloatCurve=DataInterface() {propertyValues {Curve {keys [{interpMode=RCIM_Cubic, tangentMode=RCTM_Break, time=0.6, value=1}, {interpMode=RCIM_Cubic, tangentMode=RCTM_Break, time=0.9, value=0}]}}}
        CurveIndex=$Emitter.NormalizedLoopAge
       )
       Spawn Group Mask=2
      )
      Spawn Group Mask=1
     )
    )
    ScaleA=true
    ScaleRGB=true
   ) {module=ScaleColor001}
   &SubUVAnimation(
    Random Start Frame=true
    "Sprite Renderer "=DataInterface() {propertyValues {SpriteRenderer=&self:NE_SmokeCones001_4.NiagaraSpriteRendererProperties_0}}
    SubUV Lookup Index Scale=0.5
    UseEndFrame=false
    UseStartFrame=false
   )
   &SolveForcesAndVelocity(Clamp Velocity=false, Limit Acceleration=false)
  ]
  rendererClasses [&NiagaraMeshRendererProperties, &NiagaraSpriteRendererProperties]
  renderers [
   {rendererClass=&NiagaraMeshRendererProperties, rendererIndex=0}
   {rendererClass=&NiagaraMeshRendererProperties, rendererIndex=1}
   {rendererClass=&NiagaraMeshRendererProperties, rendererIndex=2}
   {rendererClass=&NiagaraSpriteRendererProperties, rendererIndex=3}
  ]
 }
}
summary {
 systemName=NS_RoundFork_Turret_Overheat
 emitters [
  {emitterName=Glow_Decal, rendererClasses [&NiagaraSpriteRendererProperties, &NiagaraDecalRendererProperties]}
  {emitterName=Looping_Smoke, rendererClasses [&NiagaraSpriteRendererProperties]}
  {emitterName=Start_HotSmoke, rendererClasses [&NiagaraMeshRendererProperties, &NiagaraSpriteRendererProperties]}
  {emitterName=Glow_Decal001, rendererClasses [&NiagaraSpriteRendererProperties, &NiagaraDecalRendererProperties]}
  {emitterName=Looping_Smoke001, rendererClasses [&NiagaraSpriteRendererProperties]}
  {
   emitterName=Start_HotSmoke001
   rendererClasses [&NiagaraMeshRendererProperties, &NiagaraSpriteRendererProperties]
  }
 ]
 userVariables []
}
