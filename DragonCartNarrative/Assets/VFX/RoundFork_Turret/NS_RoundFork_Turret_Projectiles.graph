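# Niagara system for the RoundFork turret's looping projectile fire, with muzzle flash, sparks and impacts, timed by user fire rate and travel time.
# Projectiles and MuzzleSparks burst from parent-mesh sockets while FireProjectiles is true; projectiles lerp to the end position by normalized age.
# Impacts and Sparks delay their first loop by travel time, faking hits without collision, and ignore FireProjectiles.
# Impacts layers six spawn groups; brightness follows time of day.
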NiagaraSystem /DragonCartNarrative/Assets/VFX/RoundFork_Turret/NS_RoundFork_Turret_Projectiles
dependencies {
 usedDataInterfaces [&NiagaraDataInterfaceStaticMesh, &NiagaraDataInterfaceCurve, &NiagaraDataInterfaceVectorCurve]
 usedDynamicInputs [&LerpLinearColors, &FN_LOD_Distance_Remap]
 usedModules [
  &self:SystemSpawnScript.SystemScriptSource.SystemScriptGraph.NiagaraNodeAssignment_3.SetVariables_#1
  &self:SystemSpawnScript.SystemScriptSource.SystemScriptGraph.NiagaraNodeAssignment_1.SetVariables_#2
  &SystemState
  &self:SystemSpawnScript.SystemScriptSource.SystemScriptGraph.NiagaraNodeAssignment_0.SetVariables_#3
 ]
}
emitters {
 Impacts {
  emitterUpdateScript [
   &EmitterState(
    Delay First Loop Only=true
    Life Cycle Mode=::NewEnumerator1
    Loop Delay=$User.ProjectileTravelTime
    Loop Duration=&Divide_Float(A=1, B=$User.ProjectileFireRateSpawn)
    UseLoopDelay=true
   )
   &SpawnBurst_Instantaneous(Use Spawn Probability=false, UseLoopCountLimit=false) {module=SpawnBurst_Instantaneous001}
   &SpawnBurst_Instantaneous(Spawn Group=1, Use Spawn Probability=false, UseLoopCountLimit=false) {module=SpawnBurst_Instantaneous002}
   &SpawnBurst_Instantaneous(Spawn Group=2, Use Spawn Probability=false, UseLoopCountLimit=false) {module=SpawnBurst_Instantaneous003}
   &SpawnBurst_Instantaneous(Spawn Group=3, Use Spawn Probability=false, UseLoopCountLimit=false) {module=SpawnBurst_Instantaneous004}
   &SpawnBurst_Instantaneous(
    Spawn Count=&RandomRangeInt(Minimum=4, Maximum=8)
    Spawn Group=4
    Use Spawn Probability=false
    UseLoopCountLimit=false
   ) {module=SpawnBurst_Instantaneous005}
   &SpawnBurst_Instantaneous(Spawn Group=5, Use Spawn Probability=false, UseLoopCountLimit=false)
  ]
  particleSpawnScript [
   /SetVariables(Particles.VisibilityTag=&SpawnGroup(Max Spawn Group=5)) {isSetParameters=true}
   &InitializeParticle(
    Color=&MergeHelper_ColorSwitch(
     0=<rgba 1.6222 0.042 3 1>
     1=<rgba 1.6222 0.042 3 1>
     2=<rgba 3 1.0405 0 0.1565>
     3=<rgba 1 0.4216 0 1>
     4=<rgba 1 0.4216 0 1>
     5=<rgba 5 1.945 0.191 1>
     Index=&SpawnGroup()
     Num=::NewEnumerator4
    )
    Lifetime=&MergeHelper_FloatSwitch(
     0=0.3
     1=&RandomRangeFloat(Minimum=0.25, Maximum=0.35)
     2=0.2
     3=&RandomRangeFloat(Minimum=0.25, Maximum=0.35)
     4=&RandomRangeFloat(Minimum=0.3, Maximum=0.4)
     5=&RandomRangeFloat(Minimum=0.25, Maximum=0.35)
     Index=&SpawnGroup()
     Num=::NewEnumerator4
    )
    Mesh Scale=&MergeHelper_VectorSwitch(
     0=<0 0 0>
     1=&VectorFromFloat(Value=&RandomRangeFloat(Minimum=20, Maximum=40))
     2=<0 0 0>
     3=<0 0 0>
     4=<0 0 0>
     5=&Multiply_VectorByFloat(Vector=<0.8 1 1>, Float=&RandomRangeFloat(Minimum=35, Maximum=55))
     Index=&SpawnGroup()
     Num=::NewEnumerator4
    )
    Mesh Scale Mode=::NewEnumerator4
    Position=$System.ProjectileEndPosition
    Position Mode=::NewEnumerator1
    Sprite Size=&MergeHelper_Vector2Switch(
     0=&Vector2DFromFloat(Value=&RandomRangeFloat(Minimum=3000, Maximum=4000))
     1=<0 0>
     2=&Vector2DFromFloat(Value=&RandomRangeFloat(Minimum=3400, Maximum=4000))
     3=&Vector2DFromFloat(Value=&RandomRangeFloat(Minimum=4000, Maximum=5000))
     4=&Multiply_Vector2D_ByFloat(Float=&RandomRangeFloat(Minimum=300, Maximum=500), Vector2D=<0.2 1>)
     5=<0 0>
     Index=&SpawnGroup()
     Num=::NewEnumerator4
    )
    Sprite Size Mode=::NewEnumerator4
    UsePositionOffset=false
    Write Material Random=true
   )
   &AddVelocity(
    Cone Angle=90
    Cone Axis=&Subtract_Position(A=$Engine.Owner.Position, B=$User.ProjectileEndLocation)
    Module.UseConeFalloff=false
    Module.UseVelDistribution=false
    Override Seed=false
    Rotation Coordinate Space=::Simulation
    Velocity Mode=::NewEnumerator2
    Velocity Speed=&MergeHelper_FloatSwitch(
     0=1
     1=0
     2=1
     3=1
     4=&RandomRangeFloat(Minimum=4000, Maximum=9000)
     Index=&SpawnGroup()
     Num=::NewEnumerator3
    )
   )
   &SpriteFacingAndAlignment(
    Sprite Alignment=<0 0 1>
    Sprite Facing=&MergeHelper_VectorSwitch(
     0=&Subtract_Position(A=$Engine.Owner.Position, B=$User.ProjectileEndLocation)
     1=<0 0 0>
     2=&Subtract_Position(A=$Engine.Owner.Position, B=$User.ProjectileEndLocation)
     3=&Subtract_Position(A=$Engine.Owner.Position, B=$User.ProjectileEndLocation)
     4=<0 0 0>
     5=<0 0 0>
     Index=&SpawnGroup()
     Num=::NewEnumerator4
    )
    WriteAlignment=false
    WriteFacing=true
   )
   &InitialMeshOrientation(
    Mesh Orientation Mode=::NewEnumerator2
    Orientation Vector=&Subtract_Position(A=$Engine.Owner.Position, B=$User.ProjectileEndLocation)
    Use Rotation Vector=false
   )
  ]
  particleUpdateScript [
   &ParticleState()
   &ScaleSpriteSize(
    Scale Sprite Size Mode=::NewEnumerator0
    Uniform Scale Factor=&MergeHelper_FloatSwitch(
     0=&FloatFromCurve(
      FloatCurve=DataInterface() {propertyValues {Curve {keys [
          {arriveTangent=13.251, interpMode=RCIM_Cubic, leaveTangent=13.251, tangentMode=RCTM_User, time=0, value=0.5}
          {arriveTangent=-1.9983, interpMode=RCIM_Cubic, leaveTangent=-1.9983, tangentMode=RCTM_User, time=0.11, value=1}
          <at 0.9986 0>
         ]}}}
      Scale Curve=&RandomRangeFloat(Minimum=0.5, Maximum=1)
     )
     1=0
     2=&FloatFromCurve(
      FloatCurve=DataInterface() {propertyValues {Curve {
         defaultValue=1
         keys [
          {interpMode=RCIM_Cubic, tangentMode=RCTM_User, time=0, value=0}
          {interpMode=RCIM_Cubic, tangentMode=RCTM_User, time=0.5, value=1}
          {tangentMode=RCTM_User, time=1, value=0}
         ]
        }}}
      Scale Curve=&RandomRangeFloat(Minimum=0.5, Maximum=1)
     )
     3=&FloatFromCurve(
      FloatCurve=DataInterface() {propertyValues {Curve {
         defaultValue=0
         keys [
          {arriveTangent=4.1785, interpMode=RCIM_Cubic, leaveTangent=4.1785, tangentMode=RCTM_User, time=0, value=0.1359}
          {arriveTangent=1.1179, interpMode=RCIM_Cubic, leaveTangent=1.1179, tangentMode=RCTM_User, time=0.2042, value=0.6228}
          {arriveTangent=0.0224, interpMode=RCIM_Cubic, leaveTangent=0.0224, tangentMode=RCTM_User, time=1, value=1}
         ]
        }}}
      Scale Curve=&RandomRangeFloat(Minimum=0.5, Maximum=1)
     )
     4=&FloatFromCurve(FloatCurve=DataInterface() {propertyValues {Curve {
         defaultValue=1
         keys [
          {interpMode=RCIM_Cubic, tangentMode=RCTM_User, time=0, value=0}
          {interpMode=RCIM_Cubic, tangentMode=RCTM_User, time=0.5, value=1}
          {tangentMode=RCTM_User, time=1, value=0}
         ]
        }}})
     Index=&SpawnGroup()
     Num=::NewEnumerator3
    )
   )
   &ScaleMeshSize(Scale Factor=&VectorFromFloat(Value=&FloatFromCurve(
      FloatCurve=DataInterface() {propertyValues {Curve {
         defaultValue=1
         keys [{interpMode=RCIM_Cubic, tangentMode=RCTM_User, time=0, value=0}, {interpMode=RCIM_Cubic, tangentMode=RCTM_User, time=1, value=1}]
        }}}
      Scale Curve=&RandomRangeFloat(Minimum=0.5, Maximum=1)
     )))
   &ScaleColor(
    Scale Alpha=&FloatFromCurve(
     FloatCurve=DataInterface() {propertyValues {Curve {keys [{arriveTangent=-2.4409, arriveTangentWeight=0.162, interpMode=RCIM_Cubic, leaveTangentWeight=0.2268, time=0.2, value=1}, <at 1 0>]}}}
     Scale Curve=&MergeHelper_FloatSwitch(0=2.5, 1=3, 2=0.8, 3=1, 4=0.5, 5=0.7, Index=&SpawnGroup(), Num=::NewEnumerator4)
    )
    ScaleA=true
    ScaleRGB=false
   )
   &ScaleColor(
    Scale Alpha=&FNTimeOfDayScalar(Night Scale=0.8, Morning Scale=0.9, Evening Scale=0.8)
    Scale RGB=&VectorFromFloat(Value=&FNTimeOfDayScalar(Night Scale=1, Morning Scale=1.5, Day Scale=2, Evening Scale=1))
    ScaleA=true
    ScaleRGB=true
   ) {module=ScaleColor001}
   &DynamicMaterialParameters(
    Index 0 Param 1=1
    Index 0 Param 2=1
    Index 1 Param 1=&FloatFromCurve(FloatCurve=DataInterface() {propertyValues {Curve {keys [
         {interpMode=RCIM_Cubic, tangentMode=RCTM_User, time=0.5, value=0}
         {arriveTangent=3.8908, interpMode=RCIM_Cubic, leaveTangent=3.8908, tangentMode=RCTM_User, time=1, value=1}
        ]}}})
    Index 1 Param 2=&FloatFromCurve(FloatCurve=DataInterface() {propertyValues {Curve {keys [
         {arriveTangent=-0.6134, interpMode=RCIM_Cubic, leaveTangent=-0.6134, tangentMode=RCTM_User, time=0, value=3}
         {arriveTangent=-3.248, interpMode=RCIM_Cubic, leaveTangent=-3.248, tangentMode=RCTM_User, time=1, value=0.5}
        ]}}})
    Index 1 Param 3=&FloatFromCurve(FloatCurve=DataInterface() {propertyValues {Curve {keys [
         {arriveTangent=-1.2425, interpMode=RCIM_Cubic, time=0, value=1}
         {arriveTangent=-2.1091, interpMode=RCIM_Cubic, leaveTangent=-2.1091, tangentMode=RCTM_User, time=1, value=0}
        ]}}})
    Index 1 Param 4=&FloatFromCurve(
     FloatCurve=DataInterface() {propertyValues {Curve {keys [
         {arriveTangent=2.3563, interpMode=RCIM_Cubic, leaveTangent=2.3563, tangentMode=RCTM_User, time=0, value=0}
         {tangentMode=RCTM_User, time=1, value=1}
        ]}}}
     Scale Curve=2
    )
    Index 2 Param 1=&FloatFromCurve(FloatCurve=DataInterface() {propertyValues {Curve {keys [{arriveTangent=1.8602, interpMode=RCIM_Cubic, leaveTangent=1.8602, tangentMode=RCTM_User, time=0, value=1}, <at 1 2>]}}})
    Index 2 Param 2=5
    Index 2 Param 3=3
    Index 2 Param 4=&FloatFromCurve(FloatCurve=DataInterface() {propertyValues {Curve {keys [{arriveTangent=-2.4468, interpMode=RCIM_Cubic, leaveTangent=-2.4468, tangentMode=RCTM_User, time=0, value=0.5}, <at 0.5 0>]}}})
    Module.Index0Float0Write=true
    Module.Index0Float1Write=true
    Module.Index0Float2Write=true
    Module.Index0Float3Write=true
    Module.Index1Float0Write=true
    Module.Index1Float1Write=true
    Module.Index1Float2Write=true
    Module.Index1Float3Write=true
    Module.Index2Float0Write=true
    Module.Index2Float1Write=true
    Module.Index2Float2Write=true
    Module.Index2Float3Write=true
    Module.Index3Float0Write=true
    Module.Index3Float1Write=true
    Module.Index3Float2Write=true
    Module.Index3Float3Write=true
    Write Parameter Index 1=::NewEnumerator1
    Write Parameter Index 2=::NewEnumerator1
    Write Parameter Index 3=::NewEnumerator1
   )
   /SetVariables(Particles.Glow Sprite Color:LinearColor=&MakeLinearColorFromVectorAndFloat(
     "Vector (RGB)"=&MakeVectorFromLinearColorRGB(Color=<rgba 5 0.676 0.0025 1>)
     "Float (Alpha)"=&FloatFromCurve(
      FloatCurve=DataInterface() {propertyValues {Curve {keys [
          {arriveTangent=-2.7319, interpMode=RCIM_Cubic, time=0.05, value=1}
          {arriveTangent=-1.5377, interpMode=RCIM_Cubic, leaveTangent=-1.5377, tangentMode=RCTM_User, time=0.2, value=0.55}
          <at 1 0>
         ]}}}
      Scale Curve=4
     )
    )) {isSetParameters=true}
   &SolveForcesAndVelocity(Clamp Velocity=false, Limit Acceleration=false)
  ]
  rendererClasses [&NiagaraSpriteRendererProperties, &NiagaraMeshRendererProperties]
  renderers [
   {rendererClass=&NiagaraSpriteRendererProperties}
   {rendererClass=&NiagaraMeshRendererProperties, rendererIndex=1}
   {rendererClass=&NiagaraSpriteRendererProperties, rendererIndex=2}
   {rendererClass=&NiagaraSpriteRendererProperties, rendererIndex=3}
   {rendererClass=&NiagaraSpriteRendererProperties, rendererIndex=4}
   {rendererClass=&NiagaraMeshRendererProperties, rendererIndex=5}
  ]
 }
 MuzzleSparks {
  emitterUpdateScript [
   &EmitterState(
    Enable Distance Culling=true
    Life Cycle Mode=::NewEnumerator1
    Loop Duration=&Divide_Float(A=1, B=$User.ProjectileFireRateSpawn)
    MaxDistance=15000
    Module.bUseMaxDistance=true
    Module.bUseMinDistance=false
    Scalability Mode=::NewEnumerator1
    UseLoopDelay=false
   )
   &SpawnBurst_Instantaneous(
    Spawn Count=&MakeCustomIntFromBool(Bool=$User.FireProjectiles, True INT=&RandomRangeInt(Minimum=22, Maximum=28), False INT=0)
    Use Spawn Probability=false
    UseLoopCountLimit=false
   )
  ]
  particleSpawnScript [
   &PartitionParticles()
   &InitializeParticle(
    Color=<rgba 2.6906 0 4 2>
    Lifetime Max=0.4
    Lifetime Min=0.11
    Lifetime Mode=::NewEnumerator1
    Sprite Size Mode=::NewEnumerator1
    Uniform Sprite Size Max=16
    Uniform Sprite Size Min=8
    UsePositionOffset=false
    Write Material Random=true
   )
   /Niagara/Modules/Spawn/Location/StaticMeshLocation(
    MaskSpawnGroup=false
    Mesh Sampling Type:ENiagara_StaticSamplingMode=::NewEnumerator1
    Orientation Sampling:ENiagara_AttributeSamplingApplyOutput=::NewEnumerator1
    SamplePosOffset=true
    Sampled Position Offset:Vector3f=<-100 0 0>
    Socket Index:NiagaraInt32=&ModuloInt(A=$Emitter.LoopCount, B=2)
    Socket Sampling Mode:ENiagara_StaticSocketMode=::NewEnumerator2
    Static Mesh=DataInterface() {propertyValues {
      DefaultMesh=null
      InstanceIndex=-1
      LODIndex=0
      PreviewMesh=null
      SoftSourceActor=null
      SourceMode=AttachParent
      bAllowSamplingFromStreamingLODs=false
      bCaptureTransformsPerFrame=true
      bUsePhysicsBodyVelocity=false
      FilteredSockets []
      LODIndexUserParameter {parameter {name=None, typeDefHandle {registeredTypeIndex=108}, varData []}}
      MeshParameterBinding {parameter {name=None, typeDefHandle {registeredTypeIndex=1}, varData []}}
      SectionFilter {allowedMaterialSlots []}
     }}
   )
   &AddVelocity(
    Cone Angle=40
    Cone Axis=&NormalizeVector(
     Vector To Normalize=&Subtract_Position(A=$System.ProjectileEndPosition, B=$Particles.Position)
     Safe Normalize=true
     Safe Fallback Threshold=0
     Safe Fallback Vector=<0 0 0.001>
    )
    Module.UseConeFalloff=true
    Module.UseVelDistribution=true
    Override Seed=false
    Rotation Coordinate Space=::World
    Velocity Mode=::NewEnumerator2
    Velocity Speed=&RandomRangeFloat(Minimum=1000, Maximum=7000)
   )
  ]
  particleUpdateScript [
   &ParticleState()
   &ScaleSpriteSize(
    Uniform Curve Scale=&Lerp_Float(A=1, B=2, Alpha=$System.FarDistanceScale)
    Uniform Curve Sprite Scale=DataInterface() {propertyValues {Curve {defaultValue=0, keys [<at 0.8 1 RCIM_Cubic>, {arriveTangent=-6.9195, tangentMode=RCTM_User, time=1, value=0}]}}}
   )
   &Drag(Drag=&RandomRangeFloat(Minimum=7, Maximum=10), Use Linear Drag=true, Use Rotational Drag=false)
   &VectorNoiseForce(
    Force Amount=&FloatFromCurve(
     FloatCurve=DataInterface() {propertyValues {Curve {defaultValue=1, keys [{arriveTangent=3.3549, interpMode=RCIM_Cubic, time=0, value=0}, {tangentMode=RCTM_User, time=0.3, value=1}]}}}
     Scale Curve=5000
    )
    Module.Use Random Noise Time Interval=true
    Random Noise Time Interval:NiagaraFloat=&RandomRangeFloat(Minimum=0.1, Maximum=0.2)
   )
   &SolveForcesAndVelocity(Clamp Velocity=false, Limit Acceleration=false)
   &ScaleSpriteSizeBySpeed(
    Max Scale Factor=<1 10>
    Sample Scale Factor By Curve=true
    Scale Factor Curve=DataInterface() {propertyValues {Curve {keys [
        {arriveTangent=0.3816, interpMode=RCIM_Cubic, leaveTangent=0.3816, tangentMode=RCTM_User, time=0, value=0}
        {arriveTangent=1.6712, tangentMode=RCTM_User, time=1, value=1}
       ]}}}
    Velocity Threshold=5000
   )
   &DynamicMaterialParameters(
    Index 0 Param 1=&FloatFromCurve(FloatCurve=DataInterface() {propertyValues {Curve {keys [<at 0 1 RCIM_Cubic>,<at 1 3>]}}})
    Index 0 Param 2=5
    Index 0 Param 3=2
    Module.Index0Float0Write=true
    Module.Index0Float1Write=true
    Module.Index0Float2Write=true
    Module.Index0Float3Write=false
   )
   &ScaleColor(
    Scale RGB=&Lerp_Vector(
     A=<1 1 1>
     B=<0.2 0.1 0.1>
     Alpha=&FloatFromCurve(FloatCurve=DataInterface() {propertyValues {Curve {keys [{interpMode=RCIM_Cubic, leaveTangentWeight=0.2333, tangentMode=RCTM_User, time=0.25, value=0}, {tangentMode=RCTM_User, time=0.6, value=1}]}}})
    )
    ScaleA=false
    ScaleRGB=true
   )
  ]
  rendererClasses [&NiagaraSpriteRendererProperties]
  renderers [{rendererClass=&NiagaraSpriteRendererProperties}]
 }
 Projectiles {
  emitterUpdateScript [
   &EmitterState(
    Life Cycle Mode=::NewEnumerator1
    Loop Duration=&Divide_Float(A=1, B=$User.ProjectileFireRateSpawn)
    UseLoopDelay=false
   )
   &SpawnBurst_Instantaneous(
    Spawn Count=&MakeCustomIntFromBool(Bool=$User.FireProjectiles, True INT=1, False INT=0)
    Use Spawn Probability=false
    UseLoopCountLimit=false
   )
  ]
  particleSpawnScript [
   &PartitionParticles()
   &InitializeParticle(
    Color=<rgba 0.5407 0.014 1 2>
    Lifetime=$System.ProjectileTravelTime
    Mass=25
    Mass Mode=::NewEnumerator1
    Mesh Scale=<5 3 3>
    Mesh Scale Mode=::NewEnumerator4
    Sprite Size=<200 4000>
    Sprite Size Mode=::NewEnumerator4
    UsePositionOffset=false
    Write Material Random=true
   )
   &StaticMeshLocation(
    MaskSpawnGroup=false
    Mesh Sampling Type:ENiagara_StaticSamplingMode=::NewEnumerator1
    Orientation Sampling:ENiagara_AttributeSamplingApplyOutput=::NewEnumerator1
    SamplePosOffset=false
    Socket Index:NiagaraInt32=$Particles.PartitionParticles.Partition
    Socket Sampling Mode:ENiagara_StaticSocketMode=::NewEnumerator2
    Static Mesh=DataInterface() {propertyValues {
      DefaultMesh=null
      InstanceIndex=-1
      LODIndex=0
      PreviewMesh=null
      SoftSourceActor=null
      SourceMode=AttachParent
      bAllowSamplingFromStreamingLODs=false
      bCaptureTransformsPerFrame=true
      bUsePhysicsBodyVelocity=false
      FilteredSockets []
      LODIndexUserParameter {parameter {name=None, typeDefHandle {registeredTypeIndex=108}, varData []}}
      MeshParameterBinding {parameter {name=None, typeDefHandle {registeredTypeIndex=1}, varData []}}
      SectionFilter {allowedMaterialSlots []}
     }}
   )
   &SpriteFacingAndAlignment(
    Sprite Alignment=&Multiply_VectorByFloat(Vector=&Subtract_Position(A=$Particles.Position, B=$System.ProjectileEndPosition), Float=-1)
    WriteAlignment=true
    WriteFacing=false
   )
   /SetVariables(Particles.LightPosition:NiagaraPosition=&AddVectorToPosition(
     Position=$Particles.StaticMeshLocation.SampledPosition
     Vector=&Multiply_VectorByFloat(
      Vector=&NormalizeVector(
       Vector To Normalize=&Subtract_Position(A=$System.ProjectileEndPosition, B=$Engine.Owner.Position)
       Safe Normalize=true
       Safe Fallback Threshold=0
       Safe Fallback Vector=<0 0 0.001>
      )
      Float=100
     )
    )) {isSetParameters=true}
   &InitialMeshOrientation(
    Mesh Orientation Mode=::NewEnumerator6
    Orientation Quaternion:Quat4f=$StackContext.StaticMeshLocation.SampledRotation
    Rotation=&RandomRangeVector(Minimum=<0 0 0>, Maximum=<1 0 0>)
    Use Rotation Vector=true
   ) {module=InitialMeshOrientation001}
   &DynamicMaterialParameters(
    Index 0 Param 1=3
    Index 0 Param 2=1
    Index 0 Param 3=20
    Index 0 Param 4=0.25
    Index 2 Param 1=4
    Index 3 Param 1=3.5
    Index 3 Param 2=5
    Index 3 Param 3=3
    Index 3 Param 4=1
    Module.Index0Float0Write=true
    Module.Index0Float1Write=true
    Module.Index0Float2Write=true
    Module.Index0Float3Write=true
    Module.Index2Float0Write=true
    Module.Index2Float1Write=true
    Module.Index2Float2Write=true
    Module.Index2Float3Write=true
    Module.Index3Float0Write=true
    Module.Index3Float1Write=true
    Module.Index3Float2Write=false
    Module.Index3Float3Write=false
    Write Parameter Index 2=::NewEnumerator1
    Write Parameter Index 3=::NewEnumerator1
   )
  ]
  particleUpdateScript [
   &ParticleState()
   &ScaleSpriteSize(Uniform Curve Sprite Scale=DataInterface() {propertyValues {Curve {
       defaultValue=0
       keys [
        {interpMode=RCIM_Cubic, tangentMode=RCTM_User, time=0, value=0}
        {arriveTangentWeight=0.0167, interpMode=RCIM_Cubic, leaveTangentWeight=0.3, tangentMode=RCTM_User, time=0.02, value=1}
        {arriveTangentWeight=0.3038, interpMode=RCIM_Cubic, leaveTangentWeight=0.0169, tangentMode=RCTM_User, time=0.99, value=1}
        {tangentMode=RCTM_User, time=1, value=0}
       ]
      }}})
   &ScaleMeshSize(Scale Factor=&VectorFromCurve(
     VectorCurve=DataInterface() {propertyValues {
       XCurve {keys [
         {arriveTangent=13.277, interpMode=RCIM_Cubic, leaveTangent=13.277, tangentMode=RCTM_User, time=0, value=0.1}
         {arriveTangent=0.965, interpMode=RCIM_Cubic, leaveTangent=0.965, tangentMode=RCTM_User, time=0.12, value=0.9}
         {arriveTangent=0.1725, interpMode=RCIM_Cubic, leaveTangent=0.1725, tangentMode=RCTM_User, time=0.35, value=1.1}
        ]}
       YCurve {keys [
         {arriveTangent=39.8523, interpMode=RCIM_Cubic, leaveTangent=39.8523, tangentMode=RCTM_User, time=0, value=0.2}
         {arriveTangent=1.4412, interpMode=RCIM_Cubic, leaveTangent=1.4412, tangentMode=RCTM_User, time=0.05, value=1}
         {interpMode=RCIM_Cubic, leaveTangent=-2.1183, time=0.35, value=0.25}
        ]}
       ZCurve {keys [
         {arriveTangent=39.8523, interpMode=RCIM_Cubic, leaveTangent=39.8523, tangentMode=RCTM_User, time=0, value=0.2}
         {arriveTangent=1.4412, interpMode=RCIM_Cubic, leaveTangent=1.4412, tangentMode=RCTM_User, time=0.05, value=1}
         {interpMode=RCIM_Cubic, leaveTangent=-2.1183, time=0.35, value=0.25}
        ]}
      }}
     CurveIndex=$Particles.Age
     Scale Curve=&VectorFromFloat(Value=&Lerp_Float(A=1.5, B=3, Alpha=$System.FarDistanceScale))
    ))
   /SetVariables(Particles.Position=&Lerp_Vector(
     A=Linked() {linkedVariable {name=Particles.Initial.Position, type=NiagaraPosition}}
     B=Linked() {linkedVariable {name=System.ProjectileEndPosition, type=NiagaraPosition}}
     Alpha=$Particles.NormalizedAge
    )) {isSetParameters=true}
   /SetVariables(
    Particles.MuzzleFlashSize:Vector2f=&Multiply_Vector2D_ByFloat(
     Float=&FloatFromCurve(
      FloatCurve=DataInterface() {propertyValues {Curve {
         defaultValue=0
         keys [
          {arriveTangent=41.7959, interpMode=RCIM_Cubic, leaveTangent=41.7959, tangentMode=RCTM_User, time=0, value=0.5}
          {arriveTangent=1.1908, interpMode=RCIM_Cubic, leaveTangent=1.1908, tangentMode=RCTM_User, time=0.02, value=1}
          {arriveTangent=-0.0553, interpMode=RCIM_Cubic, leaveTangent=-0.0553, tangentMode=RCTM_User, time=0.3, value=0}
          {arriveTangent=-0.0405, interpMode=RCIM_Cubic, leaveTangent=-0.0405, tangentMode=RCTM_User, time=1, value=0}
         ]
        }}}
      CurveIndex=$Particles.Age
      Scale Curve=&Lerp_Float(A=0.6, B=1, Alpha=$System.FarDistanceScale)
     )
     Vector2D=<1500 1500>
    )
    Particles.MuzzleSpriteRotation:NiagaraFloat=&RandomRangeFloat(Minimum=0, Maximum=360)
    Particles.SpriteSizeMuzzleGlow:Vector2f=&Multiply_Vector2D_ByFloat(
     Float=&FloatFromCurve(
      FloatCurve=DataInterface() {propertyValues {Curve {
         defaultValue=0
         keys [
          {arriveTangent=47.9509, interpMode=RCIM_Cubic, leaveTangent=47.9509, tangentMode=RCTM_User, time=0, value=0.4}
          {arriveTangentWeight=0.0229, interpMode=RCIM_Cubic, leaveTangentWeight=0.1494, tangentMode=RCTM_User, time=0.0409, value=1.3596}
          {arriveTangent=-5.9613, interpMode=RCIM_Cubic, leaveTangent=-5.9613, tangentMode=RCTM_User, time=0.2307, value=0.8}
          {arriveTangent=-0.0405, interpMode=RCIM_Cubic, leaveTangent=-0.0405, tangentMode=RCTM_User, time=0.35, value=0}
         ]
        }}}
      CurveIndex=$Particles.Age
      Scale Curve=&Lerp_Float(A=0.5, B=1, Alpha=$System.FarDistanceScale)
     )
     Vector2D=<1200 2800>
    )
   ) {isSetParameters=true}
   /SetVariables(Particles.Light Color:LinearColor=&MakeLinearColorFromVectorAndFloat(
     "Vector (RGB)"=&MakeVectorFromLinearColorRGB(Color=<rgba 19.214 0.0109 40 1>)
     "Float (Alpha)"=&FloatFromCurve(
      FloatCurve=DataInterface() {propertyValues {Curve {
         defaultValue=1
         keys [
          {arriveTangent=146.4689, interpMode=RCIM_Cubic, leaveTangent=146.4689, tangentMode=RCTM_User, time=0, value=0}
          {arriveTangent=-2.5249, interpMode=RCIM_Cubic, leaveTangent=-2.5249, tangentMode=RCTM_User, time=0.025, value=1}
          {arriveTangent=-0.0934, interpMode=RCIM_Cubic, leaveTangent=-0.0934, tangentMode=RCTM_User, time=0.25, value=0}
          {arriveTangent=-0.0405, interpMode=RCIM_Cubic, leaveTangent=-0.0405, tangentMode=RCTM_User, time=1, value=0}
         ]
        }}}
      CurveIndex=$Particles.Age
     )
    )) {isSetParameters=true}
   /SetVariables(
    Particles.ProjectileGlowSize:Vector2f=&Multiply_Vector2D_ByFloat(Float=3, Vector2D=<260 1100>)
    Particles.ProjectleGlowColor:LinearColor=<rgba 0.5407 0.014 1 1>
   ) {isSetParameters=true}
   /SetVariables(
    Particles.MeshMuzzleColor:LinearColor=&MultiplyLinearColor(
     Linear Color=<rgba 2.7846 0.1562 3 0.7217>
     Scale Factor=&MakeVector4FromVectorAndFloat(
      "Vector (RGB)"=<1 1 1>
      "Float (Alpha)"=&FloatFromCurve(
       FloatCurve=DataInterface() {propertyValues {Curve {keys [
           <at 0 1 RCIM_Cubic>
           {arriveTangent=-6.1189, interpMode=RCIM_Cubic, leaveTangent=-6.1189, tangentMode=RCTM_User, time=0.15, value=0.5}
           <at 0.3 0 RCIM_Cubic>
          ]}}}
       CurveIndex=$Particles.Age
      )
     )
    )
    Particles.MuzzleMeshDyn:Vector4f=&MakeVector4(
     X=&FloatFromCurve(
      FloatCurve=DataInterface() {propertyValues {Curve {keys [
          {arriveTangent=3.5781, interpMode=RCIM_Cubic, time=0.1, value=0}
          {arriveTangent=0.3223, interpMode=RCIM_Cubic, leaveTangent=0.3223, tangentMode=RCTM_User, time=0.25, value=1}
         ]}}}
      CurveIndex=$Particles.Age
      Scale Curve=1.5
     )
     Y=1
     Z=&RandomRangeFloat(Minimum=2, Maximum=2.5)
     W=&FloatFromCurve(
      FloatCurve=DataInterface() {propertyValues {Curve {keys [
          {arriveTangent=4.0842, interpMode=RCIM_Cubic, leaveTangent=4.0842, tangentMode=RCTM_User, time=0, value=0.1}
          {arriveTangent=0.4511, interpMode=RCIM_Cubic, leaveTangent=0.4511, tangentMode=RCTM_User, time=0.35, value=1}
         ]}}}
      CurveIndex=$Particles.Age
      Scale Curve=&RandomRangeFloat(Minimum=4, Maximum=5)
     )
    )
   ) {isSetParameters=true}
   &ScaleColor(
    Scale Alpha=&FNTimeOfDayScalar(Night Scale=0.7, Morning Scale=0.9, Evening Scale=0.8)
    Scale RGB=&VectorFromFloat(Value=&FNTimeOfDayScalar(Night Scale=1, Morning Scale=1.5, Day Scale=2, Evening Scale=1))
    ScaleA=true
    ScaleRGB=true
   ) {module=ScaleColor001}
   /SetVariables(
    Particles.DarkenColor:LinearColor=&MultiplyLinearColor(
     Linear Color=<rgba 0.018 0.0192 0.02 1>
     Scale Factor=&MakeVector4(
      X=1
      Y=1
      Z=1
      W=&FloatFromCurve(
       FloatCurve=DataInterface() {propertyValues {Curve {keys [
           <at 0 1 RCIM_Cubic>
           {arriveTangent=-9.1884, interpMode=RCIM_Cubic, leaveTangent=-9.1884, tangentMode=RCTM_User, time=0.1526, value=0.5293}
           <at 0.25 0>
          ]}}}
       CurveIndex=$Particles.Age
       Scale Curve=0.7
      )
     )
    )
    Particles.SoftColor:LinearColor=&MultiplyLinearColor(
     Linear Color=$Particles.Color
     Scale Factor=&MakeVector4(
      X=1.5
      Y=1
      Z=1
      W=&FloatFromCurve(
       FloatCurve=DataInterface() {propertyValues {Curve {keys [
           <at 0 1 RCIM_Cubic>
           {arriveTangent=-9.1884, interpMode=RCIM_Cubic, leaveTangent=-9.1884, tangentMode=RCTM_User, time=0.1261, value=0.5646}
           <at 0.25 0>
          ]}}}
       CurveIndex=$Particles.Age
       Scale Curve=0.4
      )
     )
    )
   ) {isSetParameters=true}
   /SetVariables(Particles.LightRadius=&FloatFromCurve(
     FloatCurve=DataInterface() {propertyValues {Curve {keys [
         {arriveTangent=10.0647, interpMode=RCIM_Cubic, leaveTangent=10.0647, tangentMode=RCTM_User, time=0, value=0.7}
         {arriveTangentWeight=0.0336, interpMode=RCIM_Cubic, leaveTangentWeight=0.2351, tangentMode=RCTM_User, time=0.05, value=1}
         {arriveTangent=-5.0324, tangentMode=RCTM_User, time=0.4, value=0}
        ]}}}
     CurveIndex=$Particles.Age
     Scale Curve=&Lerp_Float(A=1600, B=3000, Alpha=$System.FarDistanceScale)
    )) {isSetParameters=true}
  ]
  rendererClasses [
   &NiagaraSpriteRendererProperties
   &NiagaraLightRendererProperties
   &NiagaraMeshRendererProperties
  ]
  renderers [
   {rendererClass=&NiagaraSpriteRendererProperties}
   {rendererClass=&NiagaraSpriteRendererProperties, rendererIndex=1}
   {rendererClass=&NiagaraSpriteRendererProperties, rendererIndex=2}
   {rendererClass=&NiagaraSpriteRendererProperties, rendererIndex=3}
   {rendererClass=&NiagaraLightRendererProperties, rendererIndex=4}
   {rendererClass=&NiagaraMeshRendererProperties, rendererIndex=5}
   {rendererClass=&NiagaraSpriteRendererProperties, rendererIndex=6}
  ]
 }
 Sparks {
  emitterUpdateScript [
   &EmitterState(
    Delay First Loop Only=true
    Life Cycle Mode=::NewEnumerator1
    Loop Delay=$System.ProjectileTravelTime
    Loop Duration=&Divide_Float(A=1, B=$User.ProjectileFireRateSpawn)
    UseLoopDelay=true
   )
   &SpawnBurst_Instantaneous(
    Spawn Count=&RandomRangeInt(Minimum=2, Maximum=4)
    Use Spawn Probability=false
    UseLoopCountLimit=false
   )
  ]
  particleSpawnScript [
   &InitializeParticle(
    Color=<rgba 1.6222 0.042 3 1>
    Lifetime=&RandomRangeFloat(Minimum=0.2, Maximum=0.35)
    Position=$System.ProjectileEndPosition
    Position Mode=::NewEnumerator1
    Sprite Size=&Multiply_Vector2D_ByFloat(Float=&RandomRangeFloat(Minimum=300, Maximum=500), Vector2D=<0.1 1>)
    Sprite Size Mode=::NewEnumerator4
    UsePositionOffset=false
    Write Material Random=true
   )
   &AddVelocity(
    Cone Angle=90
    Cone Axis=&Subtract_Position(A=$Engine.Owner.Position, B=$User.ProjectileEndLocation)
    Module.UseConeFalloff=false
    Module.UseVelDistribution=false
    Override Seed=false
    Rotation Coordinate Space=::Simulation
    Velocity Mode=::NewEnumerator2
    Velocity Speed=&RandomRangeFloat(Minimum=4000, Maximum=9000)
   )
  ]
  particleUpdateScript [
   &ParticleState()
   &DynamicMaterialParameters(
    Index 0 Param 1=&FloatFromCurve(FloatCurve=DataInterface() {propertyValues {Curve {keys [{arriveTangent=1.8602, interpMode=RCIM_Cubic, leaveTangent=1.8602, tangentMode=RCTM_User, time=0, value=1}, <at 1 2>]}}})
    Index 0 Param 2=5
    Index 0 Param 3=3
    Index 0 Param 4=&FloatFromCurve(FloatCurve=DataInterface() {propertyValues {Curve {keys [{arriveTangent=-2.4468, interpMode=RCIM_Cubic, leaveTangent=-2.4468, tangentMode=RCTM_User, time=0, value=0.5}, <at 0.5 0>]}}})
    Module.Param0WriteEnabled=true
    Module.Param1WriteEnabled=true
    Module.Param2WriteEnabled=true
    Module.Param3WriteEnabled=true
   )
   &Drag(
    Drag=&FloatFromCurve(FloatCurve=DataInterface() {propertyValues {Curve {
        defaultValue=0
        keys [{interpMode=RCIM_Cubic, tangentMode=RCTM_User, time=0, value=0}, {interpMode=RCIM_Cubic, tangentMode=RCTM_User, time=1, value=1}]
       }}})
    Use Linear Drag=true
    Use Rotational Drag=false
   )
   &ScaleSpriteSize(
    Scale Sprite Size Mode=::NewEnumerator0
    Uniform Scale Factor=&FloatFromCurve(
     FloatCurve=DataInterface() {propertyValues {Curve {
        defaultValue=0
        keys [{interpMode=RCIM_Cubic, tangentMode=RCTM_User, time=0, value=1}, {interpMode=RCIM_Cubic, tangentMode=RCTM_User, time=1, value=0}]
       }}}
     Scale Curve=3
    )
   )
   &ScaleColor(
    Scale Alpha=&FNTimeOfDayScalar(Night Scale=0.7, Morning Scale=0.9, Evening Scale=0.8)
    Scale RGB=&VectorFromFloat(Value=&FNTimeOfDayScalar(Night Scale=1, Morning Scale=1.5, Day Scale=2, Evening Scale=1))
    ScaleA=true
    ScaleRGB=true
   ) {module=ScaleColor001}
   &SolveForcesAndVelocity(Clamp Velocity=false, Limit Acceleration=false)
  ]
  rendererClasses [&NiagaraSpriteRendererProperties]
  renderers [{rendererClass=&NiagaraSpriteRendererProperties}]
 }
}
summary {
 systemName=NS_RoundFork_Turret_Projectiles
 emitters [
  {
   emitterName=Projectiles
   rendererClasses [
    &NiagaraSpriteRendererProperties
    &NiagaraLightRendererProperties
    &NiagaraMeshRendererProperties
   ]
  }
  {emitterName=Impacts, rendererClasses [&NiagaraSpriteRendererProperties, &NiagaraMeshRendererProperties]}
  {emitterName=Sparks, rendererClasses [&NiagaraSpriteRendererProperties]}
  {emitterName=MuzzleSparks, rendererClasses [&NiagaraSpriteRendererProperties]}
 ]
 userVariables [
  User.FireProjectiles {type=NiagaraBool, defaultValue=true}
  User.ProjectileEndLocation {type=NiagaraPosition, description="", defaultValue {}}
  User.ProjectileFireRateSpawn {type=NiagaraFloat, defaultValue=3, description=""}
  User.ProjectileTravelTime {type=NiagaraFloat, defaultValue=1, description=""}
 ]
}
